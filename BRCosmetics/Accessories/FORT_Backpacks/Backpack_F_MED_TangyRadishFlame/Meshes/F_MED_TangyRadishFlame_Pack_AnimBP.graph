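# Animation Blueprint for the TangyRadishFlame backpack, posing and simulating its wings, bows and ribbons.
# The wingPosition state machine feeds additive bone offsets driven by emote, movement, jump, crouch and skydive states, cached as WingPose.
# WingPose feeds two secondary-motion chains: ribbon trail controllers, and spring-driven AnimDynamics on bows and ribbons.
# Output selects FrontEndPose in the front end, otherwise AnimDynPose when bRigidBodyEnabled is true, else the trail pose.
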AnimBlueprint /BRCosmetics/Accessories/FORT_Backpacks/Backpack_F_MED_TangyRadishFlame/Meshes/F_MED_TangyRadishFlame_Pack_AnimBP
target_skeleton=&/F_MED_TangyRadishFlame_Pack_Skeleton
graphs [
 AnimGraph {nodes [
   Input Pose()
   n1 = Use cached pose 'trailCache'()
   n2 = Use cached pose 'AnimDynPose'()
   n3 = Get bRigidBodyEnabled()
   Output Pose(Result=Blend Poses by bool(
     BlendPose_0=Use cached pose 'FrontEndPose'()
     BlendPose_1=Blend Poses by bool(BlendPose_0=@n2, BlendPose_1=@n1, bActiveValue=@n3) {blendType=Linear}
     bActiveValue=Get bIsInFrontEnd()
    ) {blendType=Linear}) {layerGroup=None}
   AnimGraphNode_StateMachine_0 = wingPosition() {is=title AnimGraphNode_StateMachine}
   n4 = Get IsTangyRadishEmote() -> (IsTangyRadishEmote:bool)
   n5 = "Transform (Modify) Bone - Bone: dyn_wing_base_01_l"!(
    ComponentPose=Local To Component(LocalPose=@AnimGraphNode_StateMachine_0)
    bAlphaBoolEnabled:bool=@n4
    Translation="0,0,0"
    Rotation="0,30,0"
    Scale="1,1,1"
   ) -> (Pose) {
    alphaInputType=Bool
    rotationMode=BMM_Additive
    rotationSpace=BCS_BoneSpace
    translationMode=BMM_Additive
    translationSpace=BCS_BoneSpace
    alphaBoolBlend {blendInTime=0.3, blendOption=Linear, blendOutTime=0.3}
    boneToModify {boneName=dyn_wing_base_01_l}
    rotation {yaw=30}
   }
   n6 = "Transform (Modify) Bone - Bone: dyn_wing_base_01_r"!(ComponentPose=@n5, bAlphaBoolEnabled:bool=@n4, Translation="0,0,0", Rotation="0,10,10", Scale="1,1,1") -> (Pose) {
    alphaInputType=Bool
    rotationMode=BMM_Additive
    rotationSpace=BCS_BoneSpace
    translationMode=BMM_Additive
    translationSpace=BCS_BoneSpace
    alphaBoolBlend {blendInTime=0.3, blendOption=Linear, blendOutTime=0.3}
    boneToModify {boneName=dyn_wing_base_01_r}
    rotation {roll=10, yaw=10}
   }
   n7 = "Transform (Modify) Bone - Bone: base_bow_l"!(ComponentPose=@n6, bAlphaBoolEnabled:bool=@n4, Translation="-0.320,-2.072,0.349", Rotation="0,0,0", Scale="1,1,1") -> (Pose) {
    alphaInputType=Bool
    rotationMode=BMM_Additive
    rotationSpace=BCS_BoneSpace
    translationMode=BMM_Additive
    translationSpace=BCS_BoneSpace
    alphaBoolBlend {blendInTime=0.3, blendOption=Linear, blendOutTime=0.3}
    boneToModify {boneName=base_bow_l}
    rotation=<rot 0 0.0002 0>
    translation=<-0.32 -2.072 0.349>
   }
   n8 = "Transform (Modify) Bone - Bone: base_bow_r"!(ComponentPose=@n7, bAlphaBoolEnabled:bool=@n4, Translation="0,0,0", Rotation="-0.001,0.000,-0.000", Scale="1,1,1") -> (Pose) {
    alphaInputType=Bool
    rotationMode=BMM_Additive
    rotationSpace=BCS_BoneSpace
    translationMode=BMM_Additive
    translationSpace=BCS_BoneSpace
    alphaBoolBlend {blendInTime=0.3, blendOption=Linear, blendOutTime=0.3}
    boneToModify {boneName=base_bow_r}
    rotation=<rot -0.001 0 0>
   }
   n9 = "Transform (Modify) Bone - Bone: base_bow_l_loop"!(ComponentPose=@n8, bAlphaBoolEnabled:bool=@n4, Translation="-0.717,0.260,-0.000", Rotation="0,-15,0", Scale="1,1,1") -> (Pose) {
    alphaInputType=Bool
    rotationMode=BMM_Additive
    rotationSpace=BCS_BoneSpace
    translationMode=BMM_Additive
    translationSpace=BCS_BoneSpace
    alphaBoolBlend {blendInTime=0.2, blendOption=Linear, blendOutTime=1}
    boneToModify {boneName=base_bow_l_loop}
    rotation {yaw=-15}
    translation=<-0.717 0.26>
   }
   n10 = "Transform (Modify) Bone - Bone: base_bow_r_loop"!(ComponentPose=@n9, bAlphaBoolEnabled:bool=@n4, Translation="-1.738,1.005,-0.000", Rotation="-0.000,-30.000,0.000", Scale="1,1,1") -> (Pose) {
    alphaInputType=Bool
    rotationMode=BMM_Additive
    rotationSpace=BCS_BoneSpace
    translationMode=BMM_Additive
    translationSpace=BCS_BoneSpace
    alphaBoolBlend {blendInTime=0.2, blendOption=Linear, blendOutTime=1}
    boneToModify {boneName=base_bow_r_loop}
    rotation=<rot 0 -30 0>
    translation=<-1.738 1.005>
   }
   n11 = "Transform (Modify) Bone - Bone: base_bow_l"!(
    ComponentPose=@n10
    Alpha:float=Get ForwardGroundMovement() -> (ForwardGroundMovement:double)
    Translation="0,0,0"
    Rotation="15.189,-13.167,48.237"
    Scale="1,1,1"
   ) -> (Pose) {
    rotationMode=BMM_Additive
    rotationSpace=BCS_BoneSpace
    translationMode=BMM_Additive
    translationSpace=BCS_BoneSpace
    alphaBoolBlend {blendInTime=0.3, blendOption=Linear, blendOutTime=0.3}
    boneToModify {boneName=base_bow_l}
    rotation=<rot 15.189 -13.1669 48.237>
   }
   n12 = "Transform (Modify) Bone - Bone: base_bow_l"!(
    ComponentPose=@n11
    bAlphaBoolEnabled:bool=Get StandingIdle() -> (StandingIdle:bool)
    Translation="0,0,0"
    Rotation="0.000,0.000,40.000"
    Scale="1,1,1"
   ) -> (Pose) {
    alphaInputType=Bool
    rotationMode=BMM_Additive
    rotationSpace=BCS_BoneSpace
    translationMode=BMM_Additive
    translationSpace=BCS_BoneSpace
    alphaBoolBlend {blendInTime=0.3, blendOption=Linear, blendOutTime=0.3}
    boneToModify {boneName=base_bow_l}
    rotation=<rot 0 0 40>
   }
   n13 = "Transform (Modify) Bone - Bone: base_bow_l"(ComponentPose=@n12, bAlphaBoolEnabled=Get JumpingFalling() -> (JumpingFalling:bool), Translation="0,0,0", Rotation="0,0,60", Scale="1,1,1") {
    alphaInputType=Bool
    rotationMode=BMM_Additive
    rotationSpace=BCS_BoneSpace
    translationMode=BMM_Additive
    translationSpace=BCS_BoneSpace
    alphaBoolBlend {blendInTime=0.1, blendOption=Linear, blendOutTime=1}
    boneToModify {boneName=base_bow_l}
    rotation {roll=60}
   }
   n14 = Get Is Crouching or Emoting() -> (Is Crouching or Emoting:bool)
   n15 = "Transform (Modify) Bone - Bone: base_bow_l"(ComponentPose=@n13, bAlphaBoolEnabled=@n14, Translation="0,0,0.5", Rotation="0,-10,60", Scale="1,1,1") {
    alphaInputType=Bool
    rotationMode=BMM_Additive
    rotationSpace=BCS_BoneSpace
    translationMode=BMM_Additive
    translationSpace=BCS_BoneSpace
    alphaBoolBlend {blendInTime=0.3, blendOption=Linear, blendOutTime=0.3}
    boneToModify {boneName=base_bow_l}
    rotation {roll=60, yaw=-10}
    translation {z=0.5}
   }
   n16 = "Transform (Modify) Bone - Bone: base_bow_r"(ComponentPose=@n15, bAlphaBoolEnabled=@n14, Translation="0,0,-1", Rotation="0,-30,-60", Scale="1,1,1") {
    alphaInputType=Bool
    rotationMode=BMM_Additive
    rotationSpace=BCS_BoneSpace
    translationMode=BMM_Additive
    translationSpace=BCS_BoneSpace
    alphaBoolBlend {blendInTime=0.3, blendOption=Linear, blendOutTime=0.3}
    boneToModify {boneName=base_bow_r}
    rotation {roll=-60, yaw=-30}
    translation {z=-1}
   }
   n17 = "Transform (Modify) Bone - Bone: base_bow_l"!(
    ComponentPose=@n16
    Alpha:float=Get RightMovement() -> (RightMovement:double)
    Translation="0,0,0"
    Rotation="0.000,-30.000,70.000"
    Scale="1,1,1"
   ) -> (Pose) {
    rotationMode=BMM_Additive
    rotationSpace=BCS_BoneSpace
    translationMode=BMM_Additive
    translationSpace=BCS_BoneSpace
    alphaBoolBlend {blendInTime=0.3, blendOption=Linear, blendOutTime=0.3}
    alphaScaleBiasClamp {bInterpResult=true}
    boneToModify {boneName=base_bow_l}
    rotation=<rot 0 -30 70>
   }
   n18 = "Transform (Modify) Bone - Bone: base_bow_r"(
    ComponentPose=@n17
    bAlphaBoolEnabled=Get IsMovingStanding() -> (IsMovingStanding:bool)
    Translation="0,0,-1"
    Rotation="10,-60,-40"
    Scale="1,1,1"
   ) {
    alphaInputType=Bool
    rotationMode=BMM_Additive
    rotationSpace=BCS_BoneSpace
    translationMode=BMM_Additive
    translationSpace=BCS_BoneSpace
    alphaBoolBlend {blendInTime=0.3, blendOption=Linear, blendOutTime=0.3}
    boneToModify {boneName=base_bow_r}
    rotation=<rot 10 -60 -40>
    translation {z=-1}
   }
   n19 = "Transform (Modify) Bone - Bone: pack"(ComponentPose=@n18, Alpha=Get JumpingForwardAlpha() -> (JumpingForwardAlpha:double), Translation="0,0,0", Rotation="0,0,-20") {
    rotationMode=BMM_Additive
    rotationSpace=BCS_BoneSpace
    translationMode=BMM_Additive
    translationSpace=BCS_BoneSpace
    alphaBoolBlend {blendOption=Linear}
    boneToModify {boneName=pack}
    rotation {roll=-20}
   }
   n20 = Get JumpingMovement() -> (JumpingMovement:bool)
   n21 = "Transform (Modify) Bone - Bone: base_bow_l_loop"(ComponentPose=@n19, bAlphaBoolEnabled=@n20, Translation="0,0,0", Rotation="0,35,0", Scale="1,1,1") {
    alphaInputType=Bool
    rotationMode=BMM_Additive
    rotationSpace=BCS_BoneSpace
    translationMode=BMM_Additive
    translationSpace=BCS_BoneSpace
    alphaBoolBlend {blendInTime=0.2, blendOption=Linear, blendOutTime=1}
    boneToModify {boneName=base_bow_l_loop}
    rotation {yaw=35}
   }
   n22 = "Transform (Modify) Bone - Bone: base_bow_r_loop"(ComponentPose=@n21, bAlphaBoolEnabled=@n20, Translation="0,0,0", Rotation="-0.000,-30.000,-0.000", Scale="1,1,1") {
    alphaInputType=Bool
    rotationMode=BMM_Additive
    rotationSpace=BCS_BoneSpace
    translationMode=BMM_Additive
    translationSpace=BCS_BoneSpace
    alphaBoolBlend {blendInTime=0.2, blendOption=Linear, blendOutTime=1}
    boneToModify {boneName=base_bow_r_loop}
    rotation=<rot -0.0001 -30 -0.0001>
   }
   n23 = Get bIsSkyDiving()
   n24 = "Transform (Modify) Bone - Bone: base_bow_l_loop"(ComponentPose=@n22, bAlphaBoolEnabled=@n23, Translation="0,0,0", Rotation="0,10,0", Scale="1,1,1") {
    alphaInputType=Bool
    rotationMode=BMM_Additive
    rotationSpace=BCS_BoneSpace
    translationMode=BMM_Additive
    translationSpace=BCS_BoneSpace
    alphaBoolBlend {blendInTime=0.2, blendOption=Linear, blendOutTime=1}
    boneToModify {boneName=base_bow_l_loop}
    rotation {yaw=10}
   }
   n25 = "Transform (Modify) Bone - Bone: base_bow_r_loop"(ComponentPose=@n24, bAlphaBoolEnabled=@n23, Translation="0,0,0", Rotation="0,-10,0", Scale="1,1,1") {
    alphaInputType=Bool
    rotationMode=BMM_Additive
    rotationSpace=BCS_BoneSpace
    translationMode=BMM_Additive
    translationSpace=BCS_BoneSpace
    alphaBoolBlend {blendInTime=0.2, blendOption=Linear, blendOutTime=1}
    boneToModify {boneName=base_bow_r_loop}
    rotation {yaw=-10}
   }
   n26 = "Transform (Modify) Bone - Bone: dyn_ribbon_fr_L_c_01"(ComponentPose=@n25, Alpha=Get Jumping RL() -> (Jumping RL:double), Translation="0,-2,0", Rotation="0,0,0", Scale="1,1,1") {
    rotationMode=BMM_Additive
    rotationSpace=BCS_BoneSpace
    translationMode=BMM_Additive
    translationSpace=BCS_BoneSpace
    alphaBoolBlend {blendInTime=0.2, blendOption=Linear, blendOutTime=1}
    alphaScaleBiasClamp {interpSpeedDecreasing=1}
    boneToModify {boneName=dyn_ribbon_fr_L_c_01}
    translation {y=-2}
   }
   n27 = "Transform (Modify) Bone - Bone: dyn_ribbon_fr_R_a_01"!(ComponentPose=@n26, bAlphaBoolEnabled:bool=Get bIsJumping(), Translation="0,-0.5,0", Rotation="0,-25,-10", Scale="1,1,1") -> (Pose) {
    alphaInputType=Bool
    rotationMode=BMM_Additive
    rotationSpace=BCS_BoneSpace
    translationMode=BMM_Additive
    translationSpace=BCS_BoneSpace
    alphaBoolBlend {blendInTime=1, blendOption=Linear, blendOutTime=0.1}
    boneToModify {boneName=dyn_ribbon_fr_R_a_01}
    rotation {roll=-10, yaw=-25}
    translation {y=-0.5}
   }
   n28 = "Transform (Modify) Bone - Bone: dyn_ribbon_fr_R_a_01"(ComponentPose=@n27, bAlphaBoolEnabled=Get bIsFalling(), Translation="0,-0.5,0", Rotation="0,-25,-10", Scale="1,1,1") {
    alphaInputType=Bool
    rotationMode=BMM_Additive
    rotationSpace=BCS_BoneSpace
    translationMode=BMM_Additive
    translationSpace=BCS_BoneSpace
    alphaBoolBlend {blendInTime=0.1, blendOption=Linear, blendOutTime=0.1}
    boneToModify {boneName=dyn_ribbon_fr_R_a_01}
    rotation {roll=-10, yaw=-25}
    translation {y=-0.5}
   }
   n29 = "Transform (Modify) Bone - Bone: dyn_ribbon_fr_R_a_01"!(ComponentPose=@n28, Alpha:float=Get JumpingForwardAlpha(), Translation="0,0.5,0", Rotation="0,0,5", Scale="1,1,1") -> (Pose) {
    rotationMode=BMM_Additive
    rotationSpace=BCS_BoneSpace
    translationMode=BMM_Additive
    translationSpace=BCS_BoneSpace
    alphaBoolBlend {blendInTime=0.1, blendOption=Linear, blendOutTime=0.1}
    boneToModify {boneName=dyn_ribbon_fr_R_a_01}
    rotation {roll=5}
    translation {y=0.5}
   }
   n30 = "Transform (Modify) Bone - Bone: dyn_ribbon_fr_R_a_01"(
    ComponentPose=@n29
    Alpha=float * float(A=Get LeftMovement() -> (LeftMovement:double), B=Get Jumping RL()) {is=title K2Node_PromotableOperator}
    Translation="0,0.5,0"
    Rotation="0,0,0"
    Scale="1,1,1"
   ) {
    rotationMode=BMM_Additive
    rotationSpace=BCS_BoneSpace
    translationMode=BMM_Additive
    translationSpace=BCS_BoneSpace
    alphaBoolBlend {blendInTime=0.1, blendOption=Linear, blendOutTime=0.1}
    boneToModify {boneName=dyn_ribbon_fr_R_a_01}
    translation {y=0.5}
   }
   Save cached pose 'WingPose'(Pose=Component To Local(ComponentPose=@n30)) {cacheName=WingPose}
   n31 = "Transform (Modify) Bone - Bone: dyn_ribbon_l_wing_1_01"(ComponentPose=Local To Component(LocalPose=Use cached pose 'WingPose'()), Alpha=1, Translation="0,0,0", Rotation="0,0,0", Scale="1,1,1") {rotationMode=BMM_Additive, rotationSpace=BCS_BoneSpace, alphaBoolBlend {blendOption=Linear}, boneToModify {boneName=dyn_ribbon_l_wing_1_01}}
   n32 = "Transform (Modify) Bone - Bone: dyn_ribbon_l_wing_2_01"(ComponentPose=@n31, Alpha=1, Translation="0,0,0", Rotation="0,0,0", Scale="1,1,1") {rotationMode=BMM_Additive, rotationSpace=BCS_BoneSpace, alphaBoolBlend {blendOption=Linear}, boneToModify {boneName=dyn_ribbon_l_wing_2_01}}
   n33 = "Transform (Modify) Bone - Bone: dyn_ribbon_L_01"(ComponentPose=@n32, Alpha=1, Translation="0,0,0", Rotation="0,0,0", Scale="1,1,1") {rotationMode=BMM_Additive, rotationSpace=BCS_BoneSpace, alphaBoolBlend {blendOption=Linear}, boneToModify {boneName=dyn_ribbon_L_01}}
   n34 = "Transform (Modify) Bone - Bone: dyn_ribbon_l_wing_1_01"(ComponentPose=@n33, Alpha=1, Translation="0,0,0", Rotation="0,0,0", Scale="1,1,1") {rotationMode=BMM_Additive, rotationSpace=BCS_BoneSpace, alphaBoolBlend {blendOption=Linear}, boneToModify {boneName=dyn_ribbon_l_wing_1_01}}
   n35 = "Transform (Modify) Bone - Bone: dyn_ribbon_l_wing_2_01"(ComponentPose=@n34, Alpha=1, Translation="0,0,0", Rotation="0,0,0", Scale="1,1,1") {rotationMode=BMM_Additive, rotationSpace=BCS_BoneSpace, alphaBoolBlend {blendOption=Linear}, boneToModify {boneName=dyn_ribbon_l_wing_2_01}}
   n36 = "Transform (Modify) Bone - Bone: dyn_ribbon_L_01"(ComponentPose=@n35, Alpha=1, Translation="0,0,0", Rotation="0,0,0", Scale="1,1,1") {rotationMode=BMM_Additive, rotationSpace=BCS_BoneSpace, alphaBoolBlend {blendOption=Linear}, boneToModify {boneName=dyn_ribbon_L_01}}
   n37 = Trail controller - Bone: dyn_ribbon_l_wing_1_02(ComponentPose=@n36, Alpha=1) {
    bLimitRotation=true
    bUsePlanarLimit=true
    debugLifeTime=0.1
    maxDeltaTime=0.0333
    alphaScaleBiasClamp {bInterpResult=true}
    baseJoint {boneName=root}
    planarLimits [{drivingBone {boneName=pack}, planeTransform {location=<0 0 0>, rotation=<rot 0 0 -90>, scale=<1 1 1>}}]
    rotationLimits [{limitMax=<0 10 0>, limitMin=<0 -10 -20>}, {limitMax=<0 10 0>, limitMin=<0 -10 -40>}]
    rotationOffsets [<0 0 0>, <0 0 0>]
    trailBone {boneName=dyn_ribbon_l_wing_1_02}
    trailRelaxationSpeed {editorCurveData {keys [<at 0 23 RCIM_Linear>,<at 0.284 19.3799 RCIM_Linear>,<at 0.6554 8.5924 RCIM_Linear>,<at 1 5 RCIM_Linear>]}}
   }
   n38 = Trail controller - Bone: dyn_ribbon_l_wing_2_02(ComponentPose=@n37, Alpha=1) {
    bLimitRotation=true
    bUsePlanarLimit=true
    debugLifeTime=0.1
    maxDeltaTime=0.0333
    alphaScaleBiasClamp {bInterpResult=true}
    baseJoint {boneName=root}
    planarLimits [{drivingBone {boneName=pack}, planeTransform {location=<0 0 0>, rotation=<rot 0 0 -90>, scale=<1 1 1>}}]
    rotationLimits [{limitMax=<0 10 0>, limitMin=<0 -10 -20>}, {limitMax=<0 10 0>, limitMin=<0 -10 -40>}]
    rotationOffsets [<0 0 0>, <0 0 0>]
    trailBone {boneName=dyn_ribbon_l_wing_2_02}
    trailRelaxationSpeed {editorCurveData {keys [<at 0 23 RCIM_Linear>,<at 0.284 19.3799 RCIM_Linear>,<at 0.6554 8.5924 RCIM_Linear>,<at 1 5 RCIM_Linear>]}}
   }
   n39 = Trail controller - Bone: dyn_ribbon_l_wing_1_02(ComponentPose=@n38, Alpha=1) {
    bLimitRotation=true
    bUsePlanarLimit=true
    debugLifeTime=0.1
    maxDeltaTime=0.0333
    alphaScaleBiasClamp {bInterpResult=true}
    baseJoint {boneName=root}
    planarLimits [{drivingBone {boneName=pack}, planeTransform {location=<0 0 0>, rotation=<rot 0 0 -90>, scale=<1 1 1>}}]
    rotationLimits [{limitMax=<0 10 0>, limitMin=<0 -10 -20>}, {limitMax=<0 10 0>, limitMin=<0 -10 -40>}]
    rotationOffsets [<0 0 0>, <0 0 0>]
    trailBone {boneName=dyn_ribbon_l_wing_1_02}
    trailRelaxationSpeed {editorCurveData {keys [<at 0 23 RCIM_Linear>,<at 0.284 19.3799 RCIM_Linear>,<at 0.6554 8.5924 RCIM_Linear>,<at 1 5 RCIM_Linear>]}}
   }
   n40 = Trail controller - Bone: dyn_ribbon_l_wing_2_02(ComponentPose=@n39, Alpha=1) {
    bLimitRotation=true
    bUsePlanarLimit=true
    debugLifeTime=0.1
    maxDeltaTime=0.0333
    alphaScaleBiasClamp {bInterpResult=true}
    baseJoint {boneName=root}
    planarLimits [{drivingBone {boneName=pack}, planeTransform {location=<0 0 0>, rotation=<rot 0 0 -90>, scale=<1 1 1>}}]
    rotationLimits [{limitMax=<0 10 0>, limitMin=<0 -10 -20>}, {limitMax=<0 10 0>, limitMin=<0 -10 -40>}]
    rotationOffsets [<0 0 0>, <0 0 0>]
    trailBone {boneName=dyn_ribbon_l_wing_2_02}
    trailRelaxationSpeed {editorCurveData {keys [<at 0 23 RCIM_Linear>,<at 0.284 19.3799 RCIM_Linear>,<at 0.6554 8.5924 RCIM_Linear>,<at 1 5 RCIM_Linear>]}}
   }
   n41 = Trail controller - Bone: dyn_ribbon_L_02(ComponentPose=@n40, Alpha=1) {
    bLimitRotation=true
    bUsePlanarLimit=true
    debugLifeTime=0.1
    maxDeltaTime=0.0333
    alphaScaleBiasClamp {bInterpResult=true}
    baseJoint {boneName=root}
    planarLimits [{drivingBone {boneName=pack}, planeTransform {location=<0 0 0>, rotation=<rot 0 0 -90>, scale=<1 1 1>}}]
    rotationLimits [{limitMax=<0 10 0>, limitMin=<0 -10 -20>}, {limitMax=<0 10 0>, limitMin=<0 -10 -40>}]
    rotationOffsets [<0 0 0>, <0 0 0>]
    trailBone {boneName=dyn_ribbon_L_02}
    trailRelaxationSpeed {editorCurveData {keys [<at 0 23 RCIM_Linear>,<at 0.284 19.3799 RCIM_Linear>,<at 0.6554 8.5924 RCIM_Linear>,<at 1 5 RCIM_Linear>]}}
   }
   n42 = "Transform (Modify) Bone - Bone: dyn_ribbon_fr_L_a_01"(ComponentPose=@n41, Alpha=1, Translation="0,0,0", Rotation="0,0,0", Scale="1,1,1") {rotationMode=BMM_Additive, rotationSpace=BCS_BoneSpace, alphaBoolBlend {blendOption=Linear}, boneToModify {boneName=dyn_ribbon_fr_L_a_01}}
   n43 = "Transform (Modify) Bone - Bone: dyn_ribbon_fr_L_b_01"(ComponentPose=@n42, Alpha=1, Translation="0,0,0", Rotation="0,0,0", Scale="1,1,1") {rotationMode=BMM_Additive, rotationSpace=BCS_BoneSpace, alphaBoolBlend {blendOption=Linear}, boneToModify {boneName=dyn_ribbon_fr_L_b_01}}
   n44 = "Transform (Modify) Bone - Bone: dyn_ribbon_fr_L_c_01"(ComponentPose=@n43, Alpha=1, Translation="0,0,0", Rotation="0,0,0", Scale="1,1,1") {rotationMode=BMM_Additive, rotationSpace=BCS_BoneSpace, alphaBoolBlend {blendOption=Linear}, boneToModify {boneName=dyn_ribbon_fr_L_c_01}}
   n45 = Trail controller - Bone: dyn_ribbon_fr_L_a_02(ComponentPose=@n44, Alpha=1) {
    bLimitRotation=true
    bUsePlanarLimit=true
    debugLifeTime=0.1
    maxDeltaTime=0.0333
    alphaScaleBiasClamp {bInterpResult=true}
    baseJoint {boneName=root}
    planarLimits [{drivingBone {boneName=pack}, planeTransform {location=<0 0 0>, rotation=<rot 0 0 -90>, scale=<1 1 1>}}]
    rotationLimits [{limitMax=<0 10 0>, limitMin=<0 -10 -20>}, {limitMax=<0 10 0>, limitMin=<0 -10 -30>}]
    rotationOffsets [<0 0 0>, <0 0 0>]
    trailBone {boneName=dyn_ribbon_fr_L_a_02}
    trailRelaxationSpeed {editorCurveData {keys [<at 0 23 RCIM_Linear>,<at 0.284 19.3799 RCIM_Linear>,<at 0.6554 8.5924 RCIM_Linear>,<at 1 5 RCIM_Linear>]}}
   }
   n46 = Trail controller - Bone: dyn_ribbon_fr_L_b_03(ComponentPose=@n45, Alpha=1) {
    bLimitRotation=true
    bUsePlanarLimit=true
    chainLength=3
    debugLifeTime=0.1
    maxDeltaTime=0.0333
    alphaScaleBiasClamp {bInterpResult=true}
    baseJoint {boneName=root}
    planarLimits [{drivingBone {boneName=pack}, planeTransform {location=<0 0 0>, rotation=<rot 0 0 -90>, scale=<1 1 1>}}]
    rotationLimits [{limitMax=<0 10 0>, limitMin=<0 -10 -20>}, {limitMax=<0 10 0>, limitMin=<0 -10 -30>}, {limitMax=<0 10 0>, limitMin=<0 -10 -40>}]
    rotationOffsets [<0 0 0>, <0 0 0>, <0 0 0>]
    trailBone {boneName=dyn_ribbon_fr_L_b_03}
    trailRelaxationSpeed {editorCurveData {keys [<at 0 23 RCIM_Linear>,<at 0.284 19.3799 RCIM_Linear>,<at 0.6554 8.5924 RCIM_Linear>,<at 1 5 RCIM_Linear>]}}
   }
   n47 = Trail controller - Bone: dyn_ribbon_fr_L_c_03(ComponentPose=@n46, Alpha=1) {
    bLimitRotation=true
    bUsePlanarLimit=true
    chainLength=3
    debugLifeTime=0.1
    maxDeltaTime=0.0333
    alphaScaleBiasClamp {bInterpResult=true}
    baseJoint {boneName=root}
    planarLimits [{drivingBone {boneName=pack}, planeTransform {location=<0 0 0>, rotation=<rot 0 0 -90>, scale=<1 1 1>}}]
    rotationLimits [{limitMax=<0 10 0>, limitMin=<0 -10 -20>}, {limitMax=<0 10 0>, limitMin=<0 -10 -30>}, {limitMax=<0 10 0>, limitMin=<0 -10 -40>}]
    rotationOffsets [<0 0 0>, <0 0 0>, <0 0 0>]
    trailBone {boneName=dyn_ribbon_fr_L_c_03}
    trailRelaxationSpeed {editorCurveData {keys [<at 0 23 RCIM_Linear>,<at 0.284 19.3799 RCIM_Linear>,<at 0.6554 8.5924 RCIM_Linear>,<at 1 5 RCIM_Linear>]}}
   }
   n48 = "Transform (Modify) Bone - Bone: dyn_ribbon_r_wing_1_01"(ComponentPose=@n47, Alpha=1, Translation="0,0,0", Rotation="0,0,0", Scale="1,1,1") {rotationMode=BMM_Additive, rotationSpace=BCS_BoneSpace, alphaBoolBlend {blendOption=Linear}, boneToModify {boneName=dyn_ribbon_r_wing_1_01}}
   n49 = "Transform (Modify) Bone - Bone: dyn_ribbon_r_wing_2_01"(ComponentPose=@n48, Alpha=1, Translation="0,0,0", Rotation="0,0,0", Scale="1,1,1") {rotationMode=BMM_Additive, rotationSpace=BCS_BoneSpace, alphaBoolBlend {blendOption=Linear}, boneToModify {boneName=dyn_ribbon_r_wing_2_01}}
   n50 = "Transform (Modify) Bone - Bone: dyn_ribbon_r_wing_3_01"(ComponentPose=@n49, Alpha=1, Translation="0,0,0", Rotation="0,0,0", Scale="1,1,1") {rotationMode=BMM_Additive, rotationSpace=BCS_BoneSpace, alphaBoolBlend {blendOption=Linear}, boneToModify {boneName=dyn_ribbon_r_wing_3_01}}
   n51 = "Transform (Modify) Bone - Bone: dyn_ribbon_fr_R_a_01"(ComponentPose=@n50, Alpha=1, Translation="0,0,0", Rotation="0,0,0", Scale="1,1,1") {rotationMode=BMM_Additive, rotationSpace=BCS_BoneSpace, alphaBoolBlend {blendOption=Linear}, boneToModify {boneName=dyn_ribbon_fr_R_a_01}}
   n52 = "Transform (Modify) Bone - Bone: dyn_ribbon_fr_R_b_01"(ComponentPose=@n51, Alpha=1, Translation="0,0,0", Rotation="0,0,0", Scale="1,1,1") {rotationMode=BMM_Additive, rotationSpace=BCS_BoneSpace, alphaBoolBlend {blendOption=Linear}, boneToModify {boneName=dyn_ribbon_fr_R_b_01}}
   n53 = "Transform (Modify) Bone - Bone: dyn_ribbon_fr_R_c_01"(ComponentPose=@n52, Alpha=1, Translation="0,0,0", Rotation="0,0,0", Scale="1,1,1") {rotationMode=BMM_Additive, rotationSpace=BCS_BoneSpace, alphaBoolBlend {blendOption=Linear}, boneToModify {boneName=dyn_ribbon_fr_R_c_01}}
   n54 = Trail controller - Bone: dyn_ribbon_r_wing_1_02(ComponentPose=@n53, Alpha=1) {
    bLimitRotation=true
    bUsePlanarLimit=true
    debugLifeTime=0.1
    maxDeltaTime=0.0333
    alphaScaleBiasClamp {bInterpResult=true}
    baseJoint {boneName=root}
    planarLimits [{drivingBone {boneName=pack}, planeTransform {location=<0 0 0>, rotation=<rot 0 0 -90>, scale=<1 1 1>}}]
    rotationLimits [{limitMax=<0 10 0>, limitMin=<0 -10 -20>}, {limitMax=<0 10 0>, limitMin=<0 -10 -30>}]
    rotationOffsets [<0 0 0>, <0 0 0>]
    trailBone {boneName=dyn_ribbon_r_wing_1_02}
    trailRelaxationSpeed {editorCurveData {keys [<at 0 23 RCIM_Linear>,<at 0.284 19.3799 RCIM_Linear>,<at 0.6554 8.5924 RCIM_Linear>,<at 1 5 RCIM_Linear>]}}
   }
   n55 = Trail controller - Bone: dyn_ribbon_r_wing_2_02(ComponentPose=@n54, Alpha=1) {
    bLimitRotation=true
    bUsePlanarLimit=true
    debugLifeTime=0.1
    maxDeltaTime=0.0333
    alphaScaleBiasClamp {bInterpResult=true}
    baseJoint {boneName=root}
    planarLimits [{drivingBone {boneName=pack}, planeTransform {location=<0 0 0>, rotation=<rot 0 0 -90>, scale=<1 1 1>}}]
    rotationLimits [{limitMax=<0 10 20>, limitMin=<0 -10 -20>}, {limitMax=<0 10 20>, limitMin=<0 -10 -20>}]
    rotationOffsets [<0 0 0>, <0 0 0>]
    trailBone {boneName=dyn_ribbon_r_wing_2_02}
    trailRelaxationSpeed {editorCurveData {keys [<at 0 23 RCIM_Linear>,<at 0.284 19.3799 RCIM_Linear>,<at 0.6554 8.5924 RCIM_Linear>,<at 1 5 RCIM_Linear>]}}
   }
   n56 = Trail controller - Bone: dyn_ribbon_r_wing_3_02(ComponentPose=@n55, Alpha=1) {
    bLimitRotation=true
    bUsePlanarLimit=true
    debugLifeTime=0.1
    maxDeltaTime=0.0333
    alphaScaleBiasClamp {bInterpResult=true}
    baseJoint {boneName=root}
    planarLimits [{drivingBone {boneName=pack}, planeTransform {location=<0 0 0>, rotation=<rot 0 0 -90>, scale=<1 1 1>}}]
    rotationLimits [{limitMax=<0 10 0>, limitMin=<0 -10 -20>}, {limitMax=<0 10 0>, limitMin=<0 -10 -20>}]
    rotationOffsets [<0 0 0>, <0 0 0>]
    trailBone {boneName=dyn_ribbon_r_wing_3_02}
    trailRelaxationSpeed {editorCurveData {keys [<at 0 23 RCIM_Linear>,<at 0.284 19.3799 RCIM_Linear>,<at 0.6554 8.5924 RCIM_Linear>,<at 1 5 RCIM_Linear>]}}
   }
   n57 = Trail controller - Bone: dyn_ribbon_fr_R_a_03(ComponentPose=@n56, Alpha=1) {
    bLimitRotation=true
    bUsePlanarLimit=true
    chainLength=3
    debugLifeTime=0.1
    maxDeltaTime=0.0333
    alphaScaleBiasClamp {bInterpResult=true}
    baseJoint {boneName=root}
    planarLimits [{drivingBone {boneName=pack}, planeTransform {location=<0 0 0>, rotation=<rot 0 0 -90>, scale=<1 1 1>}}]
    rotationLimits [{limitMax=<0 10 0>, limitMin=<0 -10 -20>}, {limitMax=<0 10 0>, limitMin=<0 -10 -20>}, {limitMax=<180 180 180>}]
    rotationOffsets [<0 0 0>, <0 0 0>, <0 0 0>]
    trailBone {boneName=dyn_ribbon_fr_R_a_03}
    trailRelaxationSpeed {editorCurveData {keys [<at 0 23 RCIM_Linear>,<at 0.284 19.3799 RCIM_Linear>,<at 0.6554 8.5924 RCIM_Linear>,<at 1 5 RCIM_Linear>]}}
   }
   n58 = Trail controller - Bone: dyn_ribbon_fr_R_b_03(ComponentPose=@n57, Alpha=1) {
    bLimitRotation=true
    bUsePlanarLimit=true
    chainLength=3
    debugLifeTime=0.1
    maxDeltaTime=0.0333
    alphaScaleBiasClamp {bInterpResult=true}
    baseJoint {boneName=root}
    planarLimits [{drivingBone {boneName=pack}, planeTransform {location=<0 0 0>, rotation=<rot 0 0 -90>, scale=<1 1 1>}}]
    rotationLimits [{limitMax=<0 10 0>, limitMin=<0 -10 -20>}, {limitMax=<0 10 0>, limitMin=<0 -10 0>}, {limitMax=<180 180 180>}]
    rotationOffsets [<0 0 0>, <0 0 0>, <0 0 0>]
    trailBone {boneName=dyn_ribbon_fr_R_b_03}
    trailRelaxationSpeed {editorCurveData {keys [<at 0 23 RCIM_Linear>,<at 0.284 19.3799 RCIM_Linear>,<at 0.6554 8.5924 RCIM_Linear>,<at 1 5 RCIM_Linear>]}}
   }
   n59 = Trail controller - Bone: dyn_ribbon_fr_R_c_03(ComponentPose=@n58, Alpha=1) {
    bLimitRotation=true
    bUsePlanarLimit=true
    chainLength=3
    debugLifeTime=0.1
    maxDeltaTime=0.0333
    alphaScaleBiasClamp {bInterpResult=true}
    baseJoint {boneName=root}
    planarLimits [{drivingBone {boneName=pack}, planeTransform {location=<0 0 0>, rotation=<rot 0 0 -90>, scale=<1 1 1>}}]
    rotationLimits [{limitMax=<0 10 0>, limitMin=<0 -10 -20>}, {limitMax=<0 10 0>, limitMin=<0 -10 -20>}, {limitMax=<180 180 180>}]
    rotationOffsets [<0 0 0>, <0 0 0>, <0 0 0>]
    trailBone {boneName=dyn_ribbon_fr_R_c_03}
    trailRelaxationSpeed {editorCurveData {keys [<at 0 23 RCIM_Linear>,<at 0.284 19.3799 RCIM_Linear>,<at 0.6554 8.5924 RCIM_Linear>,<at 1 5 RCIM_Linear>]}}
   }
   Save cached pose 'trailCache'(Pose=Component To Local(ComponentPose=@n59)) {cacheName=trailCache}
   n60 = Anim Dynamics - Bone: dyn_bow_l_01!(ComponentPose=Local To Component(LocalPose=Use cached pose 'WingPose'()), Alpha=1, AngularSpringConstant:float=12000, ExternalForce="0,0,0") -> (Pose) {
    angularDampingOverride=1
    angularSpringConstant=12000
    bAngularSpring=true
    bOverrideAngularDamping=true
    bUseGravityOverride=true
    gravityScale=0.35
    simulationSpace=RootRelative
    alphaBoolBlend {blendOption=Linear}
    boundBone {boneName=dyn_bow_l_01}
    componentLinearAccScale=<0.75 0.75 0.75>
    componentLinearVelScale=<0.75 0.75 0.75>
    gravityOverride {z=-980}
    physicsBodyDefinitions [{
      boundBone {boneName=dyn_bow_l_01}
      boxExtents=<25 25 25>
      constraintSetup {angularLimitsMax=<0 20 0>, angularLimitsMin=<0 -5 -50>, angularTarget=<1 0 0>}
      localJointOffset=<3 0 0>
     }]
   }
   n61 = Anim Dynamics - Bone: dyn_bow_l_02!(ComponentPose=@n60, Alpha=1, AngularSpringConstant:float=12000, ExternalForce="0,0,0") -> (Pose) {
    angularDampingOverride=1
    angularSpringConstant=12000
    bAngularSpring=true
    bOverrideAngularDamping=true
    bUseGravityOverride=true
    gravityScale=0.35
    simulationSpace=RootRelative
    alphaBoolBlend {blendOption=Linear}
    boundBone {boneName=dyn_bow_l_02}
    componentLinearAccScale=<0.75 0.75 0.75>
    componentLinearVelScale=<0.75 0.75 0.75>
    gravityOverride {z=-980}
    physicsBodyDefinitions [{
      boundBone {boneName=dyn_bow_l_02}
      boxExtents=<25 25 25>
      constraintSetup {angularLimitsMax=<0 20 20>, angularLimitsMin=<0 -5 -50>, angularTarget=<1 0 0>}
      localJointOffset=<5 0 0>
     }]
   }
   n62 = Anim Dynamics - Bone: dyn_bow_l_loop_01!(ComponentPose=@n61, Alpha=1, AngularSpringConstant:float=12000, ExternalForce="0,0,0") -> (Pose) {
    angularDampingOverride=1
    angularSpringConstant=12000
    bAngularSpring=true
    bOverrideAngularDamping=true
    bUseGravityOverride=true
    gravityScale=0.35
    simulationSpace=RootRelative
    alphaBoolBlend {blendOption=Linear}
    boundBone {boneName=dyn_bow_l_loop_01}
    componentLinearAccScale=<0.75 0.75 0.75>
    componentLinearVelScale=<0.75 0.75 0.75>
    gravityOverride {z=-980}
    physicsBodyDefinitions [{
      boundBone {boneName=dyn_bow_l_loop_01}
      boxExtents=<25 25 25>
      constraintSetup {angularLimitsMax=<10 5 10>, angularLimitsMin=<-10 -5 0>, angularTarget=<1 0 0>}
      localJointOffset=<2 0 0>
     }]
   }
   n63 = Anim Dynamics - Bone: dyn_bow_l_loop_02!(ComponentPose=@n62, Alpha=1, AngularSpringConstant:float=12000, ExternalForce="0,0,0") -> (Pose) {
    angularDampingOverride=1
    angularSpringConstant=12000
    bAngularSpring=true
    bOverrideAngularDamping=true
    bUseGravityOverride=true
    gravityScale=0.35
    simulationSpace=RootRelative
    alphaBoolBlend {blendOption=Linear}
    boundBone {boneName=dyn_bow_l_loop_02}
    componentLinearAccScale=<0.75 0.75 0.75>
    componentLinearVelScale=<0.75 0.75 0.75>
    gravityOverride {z=-980}
    physicsBodyDefinitions [{
      boundBone {boneName=dyn_bow_l_loop_02}
      boxExtents=<25 25 25>
      constraintSetup {angularLimitsMax=<10 5 10>, angularLimitsMin=<-10 -5 0>, angularTarget=<1 0 0>}
      localJointOffset=<2 0 0>
     }]
   }
   n64 = Anim Dynamics - Bone: dyn_bow_r_01!(ComponentPose=@n63, Alpha=1, AngularSpringConstant:float=12000, ExternalForce="0,0,0") -> (Pose) {
    angularDampingOverride=1
    angularSpringConstant=12000
    bAngularSpring=true
    bOverrideAngularDamping=true
    bUseGravityOverride=true
    gravityScale=0.35
    simulationSpace=RootRelative
    alphaBoolBlend {blendOption=Linear}
    boundBone {boneName=dyn_bow_r_01}
    componentLinearAccScale=<0.75 0.75 0.75>
    componentLinearVelScale=<0.75 0.75 0.75>
    gravityOverride {z=-980}
    physicsBodyDefinitions [{
      boundBone {boneName=dyn_bow_r_01}
      boxExtents=<25 25 25>
      constraintSetup {angularLimitsMax=<0 20 5>, angularLimitsMin=<0 -5 -20>, angularTarget=<1 0 0>}
      localJointOffset=<3 0 0>
     }]
   }
   n65 = Anim Dynamics - Bone: dyn_bow_r_02!(ComponentPose=@n64, Alpha=1, AngularSpringConstant:float=12000, ExternalForce="0,0,0") -> (Pose) {
    angularDampingOverride=1
    angularSpringConstant=12000
    bAngularSpring=true
    bOverrideAngularDamping=true
    bUseGravityOverride=true
    gravityScale=0.35
    simulationSpace=RootRelative
    alphaBoolBlend {blendOption=Linear}
    boundBone {boneName=dyn_bow_r_02}
    componentLinearAccScale=<0.75 0.75 0.75>
    componentLinearVelScale=<0.75 0.75 0.75>
    gravityOverride {z=-980}
    physicsBodyDefinitions [{
      boundBone {boneName=dyn_bow_r_02}
      boxExtents=<25 25 25>
      constraintSetup {angularLimitsMax=<0 20 5>, angularLimitsMin=<0 -5 -40>, angularTarget=<1 0 0>}
      localJointOffset=<5 0 0>
     }]
   }
   n66 = Anim Dynamics - Bone: dyn_bow_r_loop_01!(ComponentPose=@n65, Alpha=1, AngularSpringConstant:float=12000, ExternalForce="0,0,0") -> (Pose) {
    angularDampingOverride=1
    angularSpringConstant=12000
    bAngularSpring=true
    bOverrideAngularDamping=true
    bUseGravityOverride=true
    gravityScale=0.35
    simulationSpace=RootRelative
    alphaBoolBlend {blendOption=Linear}
    boundBone {boneName=dyn_bow_r_loop_01}
    componentLinearAccScale=<0.75 0.75 0.75>
    componentLinearVelScale=<0.75 0.75 0.75>
    gravityOverride {z=-980}
    physicsBodyDefinitions [{
      boundBone {boneName=dyn_bow_r_loop_01}
      boxExtents=<25 25 25>
      constraintSetup {angularLimitsMax=<10 5 10>, angularLimitsMin=<-10 -5 0>, angularTarget=<1 0 0>}
      localJointOffset=<2 0 0>
     }]
   }
   n67 = Anim Dynamics - Bone: dyn_bow_r_loop_02!(ComponentPose=@n66, Alpha=1, AngularSpringConstant:float=12000, ExternalForce="0,0,0") -> (Pose) {
    angularDampingOverride=1
    angularSpringConstant=12000
    bAngularSpring=true
    bOverrideAngularDamping=true
    bUseGravityOverride=true
    gravityScale=0.35
    simulationSpace=RootRelative
    alphaBoolBlend {blendOption=Linear}
    boundBone {boneName=dyn_bow_r_loop_02}
    componentLinearAccScale=<0.75 0.75 0.75>
    componentLinearVelScale=<0.75 0.75 0.75>
    gravityOverride {z=-980}
    physicsBodyDefinitions [{
      boundBone {boneName=dyn_bow_r_loop_02}
      boxExtents=<25 25 25>
      constraintSetup {angularLimitsMax=<10 5 10>, angularLimitsMin=<-10 -5 0>, angularTarget=<1 0 0>}
      localJointOffset=<2 0 0>
     }]
   }
   n68 = Anim Dynamics - Bone: dyn_ribbon_l_wing_1_01!(ComponentPose=@n67, Alpha=1, GravityOverride:Vector="0,0,-980", AngularSpringConstant:float=12000, ExternalForce="0,0,0") -> (Pose) {
    angularDampingOverride=1
    angularSpringConstant=12000
    bAngularSpring=true
    bOverrideAngularDamping=true
    bUseGravityOverride=true
    gravityScale=0.35
    simulationSpace=RootRelative
    alphaBoolBlend {blendOption=Linear}
    boundBone {boneName=dyn_ribbon_l_wing_1_01}
    componentLinearAccScale=<1 1 1>
    componentLinearVelScale=<1 1 1>
    gravityOverride {z=-980}
    physicsBodyDefinitions [{
      boundBone {boneName=dyn_ribbon_l_wing_1_01}
      boxExtents=<25 25 25>
      constraintSetup {angularLimitsMax=<0 0 20>, angularLimitsMin=<0 -60 -20>, angularTarget=<1 0 0>}
      localJointOffset=<4 0 0>
     }]
   }
   n69 = Anim Dynamics - Bone: dyn_ribbon_l_wing_1_02!(ComponentPose=@n68, Alpha=1, GravityOverride:Vector="0,0,-980", AngularSpringConstant:float=12000, ExternalForce="0,0,0") -> (Pose) {
    angularDampingOverride=1
    angularSpringConstant=12000
    bAngularSpring=true
    bOverrideAngularDamping=true
    bUseGravityOverride=true
    gravityScale=0.35
    simulationSpace=RootRelative
    alphaBoolBlend {blendOption=Linear}
    boundBone {boneName=dyn_ribbon_l_wing_1_02}
    componentLinearAccScale=<1 1 1>
    componentLinearVelScale=<1 1 1>
    gravityOverride {z=-980}
    physicsBodyDefinitions [{
      boundBone {boneName=dyn_ribbon_l_wing_1_02}
      boxExtents=<25 25 25>
      constraintSetup {angularLimitsMax=<0 0 20>, angularLimitsMin=<0 -60 -20>, angularTarget=<1 0 0>}
      localJointOffset=<5 0 0>
     }]
   }
   n70 = Anim Dynamics - Bone: dyn_ribbon_l_wing_2_01!(ComponentPose=@n69, Alpha=1, GravityOverride:Vector="0,0,-980", AngularSpringConstant:float=12000, ExternalForce="0,0,0") -> (Pose) {
    angularDampingOverride=1
    angularSpringConstant=12000
    bAngularSpring=true
    bOverrideAngularDamping=true
    bUseGravityOverride=true
    gravityScale=0.35
    simulationSpace=RootRelative
    alphaBoolBlend {blendOption=Linear}
    boundBone {boneName=dyn_ribbon_l_wing_2_01}
    componentLinearAccScale=<1 1 1>
    componentLinearVelScale=<1 1 1>
    gravityOverride {z=-980}
    physicsBodyDefinitions [{
      boundBone {boneName=dyn_ribbon_l_wing_2_01}
      boxExtents=<25 25 25>
      constraintSetup {angularLimitsMax=<0 20 10>, angularLimitsMin=<0 -20 -30>, angularTarget=<1 0 0>}
      localJointOffset=<3 0 0>
     }]
   }
   n71 = Anim Dynamics - Bone: dyn_ribbon_l_wing_2_02!(ComponentPose=@n70, Alpha=1, GravityOverride:Vector="0,0,-980", AngularSpringConstant:float=12000, ExternalForce="0,0,0") -> (Pose) {
    angularDampingOverride=1
    angularSpringConstant=12000
    bAngularSpring=true
    bOverrideAngularDamping=true
    bUseGravityOverride=true
    gravityScale=0.35
    simulationSpace=RootRelative
    alphaBoolBlend {blendOption=Linear}
    boundBone {boneName=dyn_ribbon_l_wing_2_02}
    componentLinearAccScale=<1 1 1>
    componentLinearVelScale=<1 1 1>
    gravityOverride {z=-980}
    physicsBodyDefinitions [{
      boundBone {boneName=dyn_ribbon_l_wing_2_02}
      boxExtents=<25 25 25>
      constraintSetup {angularLimitsMax=<0 20 10>, angularLimitsMin=<0 -20 -30>, angularTarget=<1 0 0>}
      localJointOffset=<5 0 0>
     }]
   }
   n72 = Anim Dynamics - Bone: dyn_ribbon_fr_L_a_01!(ComponentPose=@n71, bAlphaBoolEnabled:bool=True, GravityOverride:Vector="0,0,-980", AngularSpringConstant:float=12000, ExternalForce="0,0,0") -> (Pose) {
    alphaInputType=Bool
    angularDampingOverride=1
    angularSpringConstant=12000
    bAngularSpring=true
    bOverrideAngularDamping=true
    bUseGravityOverride=true
    gravityScale=0.35
    simulationSpace=RootRelative
    alphaBoolBlend {blendOption=Linear}
    boundBone {boneName=dyn_ribbon_fr_L_a_01}
    componentLinearAccScale=<0.75 0.75 0.75>
    componentLinearVelScale=<0.75 0.75 0.75>
    gravityOverride {z=-980}
    physicsBodyDefinitions [{
      boundBone {boneName=dyn_ribbon_fr_L_a_01}
      boxExtents=<25 25 25>
      constraintSetup {angularLimitsMax=<0 20 20>, angularLimitsMin=<0 -10 -20>, angularTarget=<1 0 0>}
      localJointOffset=<2 0 0>
     }]
   }
   n73 = Anim Dynamics - Bone: dyn_ribbon_fr_L_a_02!(ComponentPose=@n72, bAlphaBoolEnabled:bool=True, GravityOverride:Vector="0,0,-980", AngularSpringConstant:float=12000, ExternalForce="0,0,0") -> (Pose) {
    alphaInputType=Bool
    angularDampingOverride=1
    angularSpringConstant=12000
    bAngularSpring=true
    bOverrideAngularDamping=true
    bUseGravityOverride=true
    gravityScale=0.35
    simulationSpace=RootRelative
    alphaBoolBlend {blendOption=Linear}
    boundBone {boneName=dyn_ribbon_fr_L_a_02}
    componentLinearAccScale=<0.75 0.75 0.75>
    componentLinearVelScale=<0.75 0.75 0.75>
    gravityOverride {z=-980}
    physicsBodyDefinitions [{
      boundBone {boneName=dyn_ribbon_fr_L_a_02}
      boxExtents=<25 25 25>
      constraintSetup {angularLimitsMax=<0 20 20>, angularLimitsMin=<0 -10 -20>, angularTarget=<1 0 0>}
      localJointOffset=<2 0 0>
     }]
   }
   n74 = Anim Dynamics - Bone: dyn_ribbon_fr_L_b_01!(ComponentPose=@n73, bAlphaBoolEnabled:bool=True, GravityOverride:Vector="0,0,-980", AngularSpringConstant:float=12000, ExternalForce="0,0,0") -> (Pose) {
    alphaInputType=Bool
    angularDampingOverride=1
    angularSpringConstant=12000
    bAngularSpring=true
    bOverrideAngularDamping=true
    bUseGravityOverride=true
    gravityScale=0.35
    simulationSpace=RootRelative
    alphaBoolBlend {blendOption=Linear}
    boundBone {boneName=dyn_ribbon_fr_L_b_01}
    componentLinearAccScale=<0.75 0.75 0.75>
    componentLinearVelScale=<0.75 0.75 0.75>
    gravityOverride {z=-980}
    physicsBodyDefinitions [{
      boundBone {boneName=dyn_ribbon_fr_L_b_01}
      boxExtents=<25 25 25>
      constraintSetup {angularLimitsMax=<0 20 0>, angularLimitsMin=<0 -10 -20>, angularTarget=<1 0 0>}
      localJointOffset=<6 0 0>
     }]
   }
   n75 = Anim Dynamics - Bone: dyn_ribbon_fr_L_b_02!(ComponentPose=@n74, bAlphaBoolEnabled:bool=True, GravityOverride:Vector="0,0,-980", AngularSpringConstant:float=12000, ExternalForce="0,0,0") -> (Pose) {
    alphaInputType=Bool
    angularDampingOverride=1
    angularSpringConstant=12000
    bAngularSpring=true
    bOverrideAngularDamping=true
    bUseGravityOverride=true
    gravityScale=0.35
    simulationSpace=RootRelative
    alphaBoolBlend {blendOption=Linear}
    boundBone {boneName=dyn_ribbon_fr_L_b_02}
    componentLinearAccScale=<0.75 0.75 0.75>
    componentLinearVelScale=<0.75 0.75 0.75>
    gravityOverride {z=-980}
    physicsBodyDefinitions [{
      boundBone {boneName=dyn_ribbon_fr_L_b_02}
      boxExtents=<25 25 25>
      constraintSetup {angularLimitsMax=<0 20 10>, angularLimitsMin=<0 -20 -20>, angularTarget=<1 0 0>}
      localJointOffset=<6 0 0>
     }]
   }
   n76 = Anim Dynamics - Bone: dyn_ribbon_fr_L_b_03!(ComponentPose=@n75, bAlphaBoolEnabled:bool=True, GravityOverride:Vector="0,0,-980", AngularSpringConstant:float=12000, ExternalForce="0,0,0") -> (Pose) {
    alphaInputType=Bool
    angularDampingOverride=1
    angularSpringConstant=12000
    bAngularSpring=true
    bOverrideAngularDamping=true
    bUseGravityOverride=true
    gravityScale=0.35
    simulationSpace=RootRelative
    alphaBoolBlend {blendOption=Linear}
    boundBone {boneName=dyn_ribbon_fr_L_b_03}
    componentLinearAccScale=<0.75 0.75 0.75>
    componentLinearVelScale=<0.75 0.75 0.75>
    gravityOverride {z=-980}
    physicsBodyDefinitions [{
      boundBone {boneName=dyn_ribbon_fr_L_b_03}
      boxExtents=<25 25 25>
      constraintSetup {angularLimitsMax=<0 20 20>, angularLimitsMin=<0 -20 -20>, angularTarget=<1 0 0>}
      localJointOffset=<6 0 0>
     }]
   }
   n77 = Anim Dynamics - Bone: dyn_ribbon_fr_L_c_01!(ComponentPose=@n76, bAlphaBoolEnabled:bool=True, GravityOverride:Vector="0,0,-980", AngularSpringConstant:float=12000, ExternalForce="0,0,0") -> (Pose) {
    alphaInputType=Bool
    angularDampingOverride=1
    angularSpringConstant=12000
    bAngularSpring=true
    bOverrideAngularDamping=true
    bUseGravityOverride=true
    gravityScale=0.35
    simulationSpace=RootRelative
    alphaBoolBlend {blendOption=Linear}
    boundBone {boneName=dyn_ribbon_fr_L_c_01}
    componentLinearAccScale=<0.75 0.75 0.75>
    componentLinearVelScale=<0.75 0.75 0.75>
    gravityOverride {z=-980}
    physicsBodyDefinitions [{
      boundBone {boneName=dyn_ribbon_fr_L_c_01}
      boxExtents=<25 25 25>
      constraintSetup {angularLimitsMax=<0 0 0>, angularLimitsMin=<0 0 0>, angularTarget=<1 0 0>}
      localJointOffset=<0 0 0>
     }]
   }
   n78 = Anim Dynamics - Bone: dyn_ribbon_fr_L_c_02!(ComponentPose=@n77, bAlphaBoolEnabled:bool=True, GravityOverride:Vector="0,0,-980", AngularSpringConstant:float=12000, ExternalForce="0,0,0") -> (Pose) {
    alphaInputType=Bool
    angularDampingOverride=1
    angularSpringConstant=12000
    bAngularSpring=true
    bOverrideAngularDamping=true
    bUseGravityOverride=true
    gravityScale=0.35
    simulationSpace=RootRelative
    alphaBoolBlend {blendOption=Linear}
    boundBone {boneName=dyn_ribbon_fr_L_c_02}
    componentLinearAccScale=<0.75 0.75 0.75>
    componentLinearVelScale=<0.75 0.75 0.75>
    gravityOverride {z=-980}
    physicsBodyDefinitions [{
      boundBone {boneName=dyn_ribbon_fr_L_c_02}
      boxExtents=<25 25 25>
      constraintSetup {angularLimitsMax=<0 20 20>, angularLimitsMin=<0 -10 -40>, angularTarget=<1 0 0>}
      localJointOffset=<5 0 0>
     }]
   }
   n79 = Anim Dynamics - Bone: dyn_ribbon_fr_L_c_03!(ComponentPose=@n78, bAlphaBoolEnabled:bool=True, GravityOverride:Vector="0,0,-980", AngularSpringConstant:float=12000, ExternalForce="0,0,0") -> (Pose) {
    alphaInputType=Bool
    angularDampingOverride=1
    angularSpringConstant=12000
    bAngularSpring=true
    bOverrideAngularDamping=true
    bUseGravityOverride=true
    gravityScale=0.35
    simulationSpace=RootRelative
    alphaBoolBlend {blendOption=Linear}
    boundBone {boneName=dyn_ribbon_fr_L_c_03}
    componentLinearAccScale=<0.75 0.75 0.75>
    componentLinearVelScale=<0.75 0.75 0.75>
    gravityOverride {z=-980}
    physicsBodyDefinitions [{
      boundBone {boneName=dyn_ribbon_fr_L_c_03}
      boxExtents=<25 25 25>
      constraintSetup {angularLimitsMax=<0 20 20>, angularLimitsMin=<0 -10 -20>, angularTarget=<1 0 0>}
      localJointOffset=<5 0 0>
     }]
   }
   n80 = Anim Dynamics - Bone: dyn_ribbon_bk_L_a_01!(ComponentPose=@n79, Alpha=1, GravityOverride:Vector="0,0,-980", AngularSpringConstant:float=12000, ExternalForce="0,0,0") -> (Pose) {
    angularDampingOverride=1
    angularSpringConstant=12000
    bAngularSpring=true
    bOverrideAngularDamping=true
    bUseGravityOverride=true
    gravityScale=0.35
    simulationSpace=RootRelative
    alphaBoolBlend {blendOption=Linear}
    boundBone {boneName=dyn_ribbon_bk_L_a_01}
    componentLinearAccScale=<0.75 0.75 0.75>
    componentLinearVelScale=<0.75 0.75 0.75>
    gravityOverride {z=-980}
    physicsBodyDefinitions [{
      boundBone {boneName=dyn_ribbon_bk_L_a_01}
      boxExtents=<25 25 25>
      constraintSetup {angularLimitsMax=<0 0 0>, angularLimitsMin=<0 -5 0>, angularTarget=<1 0 0>}
      localJointOffset=<3 0 0>
     }]
   }
   n81 = Anim Dynamics - Bone: dyn_ribbon_bk_L_a_02!(ComponentPose=@n80, Alpha=1, GravityOverride:Vector="0,0,-980", AngularSpringConstant:float=12000, ExternalForce="0,0,0") -> (Pose) {
    angularDampingOverride=1
    angularSpringConstant=12000
    bAngularSpring=true
    bOverrideAngularDamping=true
    bUseGravityOverride=true
    gravityScale=0.35
    simulationSpace=RootRelative
    alphaBoolBlend {blendOption=Linear}
    boundBone {boneName=dyn_ribbon_bk_L_a_02}
    componentLinearAccScale=<0.75 0.75 0.75>
    componentLinearVelScale=<0.75 0.75 0.75>
    gravityOverride {z=-980}
    physicsBodyDefinitions [{
      boundBone {boneName=dyn_ribbon_bk_L_a_02}
      boxExtents=<25 25 25>
      constraintSetup {angularLimitsMax=<0 20 0>, angularLimitsMin=<0 -5 -30>, angularTarget=<1 0 0>}
      localJointOffset=<3 0 0>
     }]
   }
   n82 = Anim Dynamics - Bone: dyn_ribbon_bk_L_b_01!(ComponentPose=@n81, Alpha=1, GravityOverride:Vector="0,0,-980", AngularSpringConstant:float=12000, ExternalForce="0,0,0") -> (Pose) {
    angularDampingOverride=1
    angularSpringConstant=12000
    bAngularSpring=true
    bOverrideAngularDamping=true
    bUseGravityOverride=true
    gravityScale=0.35
    simulationSpace=RootRelative
    alphaBoolBlend {blendOption=Linear}
    boundBone {boneName=dyn_ribbon_bk_L_b_01}
    componentLinearAccScale=<0.75 0.75 0.75>
    componentLinearVelScale=<0.75 0.75 0.75>
    gravityOverride {z=-980}
    physicsBodyDefinitions [{
      boundBone {boneName=dyn_ribbon_bk_L_b_01}
      boxExtents=<25 25 25>
      constraintSetup {angularLimitsMax=<0 20 0>, angularLimitsMin=<0 -5 -30>, angularTarget=<1 0 0>}
      localJointOffset=<3 0 0>
     }]
   }
   n83 = Anim Dynamics - Bone: dyn_ribbon_bk_L_b_02!(ComponentPose=@n82, Alpha=1, GravityOverride:Vector="0,0,-980", AngularSpringConstant:float=12000, ExternalForce="0,0,0") -> (Pose) {
    angularDampingOverride=1
    angularSpringConstant=12000
    bAngularSpring=true
    bOverrideAngularDamping=true
    bUseGravityOverride=true
    gravityScale=0.35
    simulationSpace=RootRelative
    alphaBoolBlend {blendOption=Linear}
    boundBone {boneName=dyn_ribbon_bk_L_b_02}
    componentLinearAccScale=<0.75 0.75 0.75>
    componentLinearVelScale=<0.75 0.75 0.75>
    gravityOverride {z=-980}
    physicsBodyDefinitions [{
      boundBone {boneName=dyn_ribbon_bk_L_b_02}
      boxExtents=<25 25 25>
      constraintSetup {angularLimitsMax=<0 20 0>, angularLimitsMin=<0 -5 -30>, angularTarget=<1 0 0>}
      localJointOffset=<3 0 0>
     }]
   }
   n84 = Anim Dynamics - Bone: dyn_ribbon_bk_L_b_03!(ComponentPose=@n83, Alpha=1, GravityOverride:Vector="0,0,-980", AngularSpringConstant:float=12000, ExternalForce="0,0,0") -> (Pose) {
    angularDampingOverride=1
    angularSpringConstant=12000
    bAngularSpring=true
    bOverrideAngularDamping=true
    bUseGravityOverride=true
    gravityScale=0.35
    simulationSpace=RootRelative
    alphaBoolBlend {blendOption=Linear}
    boundBone {boneName=dyn_ribbon_bk_L_b_03}
    componentLinearAccScale=<0.75 0.75 0.75>
    componentLinearVelScale=<0.75 0.75 0.75>
    gravityOverride {z=-980}
    physicsBodyDefinitions [{
      boundBone {boneName=dyn_ribbon_bk_L_b_03}
      boxExtents=<25 25 25>
      constraintSetup {angularLimitsMax=<0 20 0>, angularLimitsMin=<0 -5 -30>, angularTarget=<1 0 0>}
      localJointOffset=<3 0 0>
     }]
   }
   n85 = Anim Dynamics - Bone: dyn_ribbon_bk_L_c_01!(ComponentPose=@n84, Alpha=1, GravityOverride:Vector="0,0,-980", AngularSpringConstant:float=12000, ExternalForce="0,0,0") -> (Pose) {
    angularDampingOverride=1
    angularSpringConstant=12000
    bAngularSpring=true
    bOverrideAngularDamping=true
    bUseGravityOverride=true
    gravityScale=0.35
    simulationSpace=RootRelative
    alphaBoolBlend {blendOption=Linear}
    boundBone {boneName=dyn_ribbon_bk_L_c_01}
    componentLinearAccScale=<0.75 0.75 0.75>
    componentLinearVelScale=<0.75 0.75 0.75>
    gravityOverride {z=-980}
    physicsBodyDefinitions [{
      boundBone {boneName=dyn_ribbon_bk_L_c_01}
      boxExtents=<25 25 25>
      constraintSetup {angularLimitsMax=<0 20 0>, angularLimitsMin=<0 -5 -30>, angularTarget=<1 0 0>}
      localJointOffset=<3 0 0>
     }]
   }
   n86 = Anim Dynamics - Bone: dyn_ribbon_bk_L_c_02!(ComponentPose=@n85, Alpha=1, GravityOverride:Vector="0,0,-980", AngularSpringConstant:float=12000, ExternalForce="0,0,0") -> (Pose) {
    angularDampingOverride=1
    angularSpringConstant=12000
    bAngularSpring=true
    bOverrideAngularDamping=true
    bUseGravityOverride=true
    gravityScale=0.35
    simulationSpace=RootRelative
    alphaBoolBlend {blendOption=Linear}
    boundBone {boneName=dyn_ribbon_bk_L_c_02}
    componentLinearAccScale=<0.75 0.75 0.75>
    componentLinearVelScale=<0.75 0.75 0.75>
    gravityOverride {z=-980}
    physicsBodyDefinitions [{
      boundBone {boneName=dyn_ribbon_bk_L_c_02}
      boxExtents=<25 25 25>
      constraintSetup {angularLimitsMax=<0 20 0>, angularLimitsMin=<0 -5 -30>, angularTarget=<1 0 0>}
      localJointOffset=<3 0 0>
     }]
   }
   n87 = Anim Dynamics - Bone: dyn_ribbon_bk_L_c_03!(ComponentPose=@n86, Alpha=1, GravityOverride:Vector="0,0,-980", AngularSpringConstant:float=12000, ExternalForce="0,0,0") -> (Pose) {
    angularDampingOverride=1
    angularSpringConstant=12000
    bAngularSpring=true
    bOverrideAngularDamping=true
    bUseGravityOverride=true
    gravityScale=0.35
    simulationSpace=RootRelative
    alphaBoolBlend {blendOption=Linear}
    boundBone {boneName=dyn_ribbon_bk_L_c_03}
    componentLinearAccScale=<0.75 0.75 0.75>
    componentLinearVelScale=<0.75 0.75 0.75>
    gravityOverride {z=-980}
    physicsBodyDefinitions [{
      boundBone {boneName=dyn_ribbon_bk_L_c_03}
      boxExtents=<25 25 25>
      constraintSetup {angularLimitsMax=<0 20 0>, angularLimitsMin=<0 -5 -30>, angularTarget=<1 0 0>}
      localJointOffset=<3 0 0>
     }]
   }
   n88 = Anim Dynamics - Bone: dyn_ribbon_r_wing_1_01!(ComponentPose=@n87, Alpha=1, GravityOverride:Vector="0,0,-980", AngularSpringConstant:float=12000, ExternalForce="0,0,0") -> (Pose) {
    angularDampingOverride=1
    angularSpringConstant=12000
    bAngularSpring=true
    bOverrideAngularDamping=true
    bUseGravityOverride=true
    gravityScale=0.35
    simulationSpace=RootRelative
    alphaBoolBlend {blendOption=Linear}
    boundBone {boneName=dyn_ribbon_r_wing_1_01}
    componentLinearAccScale=<1 1 1>
    componentLinearVelScale=<1 1 1>
    gravityOverride {z=-980}
    physicsBodyDefinitions [{
      boundBone {boneName=dyn_ribbon_r_wing_1_01}
      boxExtents=<25 25 25>
      constraintSetup {angularLimitsMax=<0 20 0>, angularLimitsMin=<0 -5 -30>, angularTarget=<1 0 0>}
      localJointOffset=<2 0 0>
     }]
   }
   n89 = Anim Dynamics - Bone: dyn_ribbon_r_wing_1_02!(ComponentPose=@n88, Alpha=1, GravityOverride:Vector="0,0,-980", AngularSpringConstant:float=12000, ExternalForce="0,0,0") -> (Pose) {
    angularDampingOverride=1
    angularSpringConstant=12000
    bAngularSpring=true
    bOverrideAngularDamping=true
    bUseGravityOverride=true
    gravityScale=0.35
    simulationSpace=RootRelative
    alphaBoolBlend {blendOption=Linear}
    boundBone {boneName=dyn_ribbon_r_wing_1_02}
    componentLinearAccScale=<1 1 1>
    componentLinearVelScale=<1 1 1>
    gravityOverride {z=-980}
    physicsBodyDefinitions [{
      boundBone {boneName=dyn_ribbon_r_wing_1_02}
      boxExtents=<25 25 25>
      constraintSetup {angularLimitsMax=<0 20 0>, angularLimitsMin=<0 -5 -30>, angularTarget=<1 0 0>}
      localJointOffset=<2 0 0>
     }]
   }
   n90 = Anim Dynamics - Bone: dyn_ribbon_r_wing_2_01!(ComponentPose=@n89, Alpha=1, GravityOverride:Vector="0,0,-980", AngularSpringConstant:float=12000, ExternalForce="0,0,0") -> (Pose) {
    angularDampingOverride=1
    angularSpringConstant=12000
    bAngularSpring=true
    bOverrideAngularDamping=true
    bUseGravityOverride=true
    gravityScale=0.35
    simulationSpace=RootRelative
    alphaBoolBlend {blendOption=Linear}
    boundBone {boneName=dyn_ribbon_r_wing_2_01}
    componentLinearAccScale=<1 1 1>
    componentLinearVelScale=<1 1 1>
    gravityOverride {z=-980}
    physicsBodyDefinitions [{
      boundBone {boneName=dyn_ribbon_r_wing_2_01}
      boxExtents=<25 25 25>
      constraintSetup {angularLimitsMax=<0 20 10>, angularLimitsMin=<0 -5 -30>, angularTarget=<1 0 0>}
      localJointOffset=<2 0 0>
     }]
   }
   n91 = Anim Dynamics - Bone: dyn_ribbon_r_wing_2_02!(ComponentPose=@n90, Alpha=1, GravityOverride:Vector="0,0,-980", AngularSpringConstant:float=12000, ExternalForce="0,0,0") -> (Pose) {
    angularDampingOverride=1
    angularSpringConstant=12000
    bAngularSpring=true
    bOverrideAngularDamping=true
    bUseGravityOverride=true
    gravityScale=0.35
    simulationSpace=RootRelative
    alphaBoolBlend {blendOption=Linear}
    boundBone {boneName=dyn_ribbon_r_wing_2_02}
    componentLinearAccScale=<1 1 1>
    componentLinearVelScale=<1 1 1>
    gravityOverride {z=-980}
    physicsBodyDefinitions [{
      boundBone {boneName=dyn_ribbon_r_wing_2_02}
      boxExtents=<25 25 25>
      constraintSetup {angularLimitsMax=<0 20 10>, angularLimitsMin=<0 -5 -30>, angularTarget=<1 0 0>}
      localJointOffset=<2 0 0>
     }]
   }
   n92 = Anim Dynamics - Bone: dyn_ribbon_r_wing_3_01!(ComponentPose=@n91, Alpha=1, GravityOverride:Vector="0,0,-980", AngularSpringConstant:float=12000, ExternalForce="0,0,0") -> (Pose) {
    angularDampingOverride=1
    angularSpringConstant=12000
    bAngularSpring=true
    bOverrideAngularDamping=true
    bUseGravityOverride=true
    gravityScale=0.35
    simulationSpace=RootRelative
    alphaBoolBlend {blendOption=Linear}
    boundBone {boneName=dyn_ribbon_r_wing_3_01}
    componentLinearAccScale=<1 1 1>
    componentLinearVelScale=<1 1 1>
    gravityOverride {z=-980}
    physicsBodyDefinitions [{
      boundBone {boneName=dyn_ribbon_r_wing_3_01}
      boxExtents=<25 25 25>
      constraintSetup {angularLimitsMax=<0 20 10>, angularLimitsMin=<0 -5 -30>, angularTarget=<1 0 0>}
      localJointOffset=<2 0 0>
     }]
   }
   n93 = Anim Dynamics - Bone: dyn_ribbon_r_wing_3_02!(ComponentPose=@n92, Alpha=1, GravityOverride:Vector="0,0,-980", AngularSpringConstant:float=12000, ExternalForce="0,0,0") -> (Pose) {
    angularDampingOverride=1
    angularSpringConstant=12000
    bAngularSpring=true
    bOverrideAngularDamping=true
    bUseGravityOverride=true
    gravityScale=0.35
    simulationSpace=RootRelative
    alphaBoolBlend {blendOption=Linear}
    boundBone {boneName=dyn_ribbon_r_wing_3_02}
    componentLinearAccScale=<1 1 1>
    componentLinearVelScale=<1 1 1>
    gravityOverride {z=-980}
    physicsBodyDefinitions [{
      boundBone {boneName=dyn_ribbon_r_wing_3_02}
      boxExtents=<25 25 25>
      constraintSetup {angularLimitsMax=<0 20 10>, angularLimitsMin=<0 -5 -30>, angularTarget=<1 0 0>}
      localJointOffset=<2 0 0>
     }]
   }
   n94 = Anim Dynamics - Bone: dyn_ribbon_fr_R_a_01!(ComponentPose=@n93, bAlphaBoolEnabled:bool=True, GravityOverride:Vector="0,0,-980", AngularSpringConstant:float=15000, ExternalForce="0,0,0") -> (Pose) {
    alphaInputType=Bool
    angularDampingOverride=1
    angularSpringConstant=15000
    bAngularSpring=true
    bOverrideAngularDamping=true
    bUseGravityOverride=true
    gravityScale=0.35
    simulationSpace=RootRelative
    alphaBoolBlend {blendOption=Linear}
    boundBone {boneName=dyn_ribbon_fr_R_a_01}
    componentLinearAccScale=<0.75 0.75 0.75>
    componentLinearVelScale=<0.75 0.75 0.75>
    gravityOverride {z=-980}
    physicsBodyDefinitions [{
      boundBone {boneName=dyn_ribbon_fr_R_a_01}
      boxExtents=<25 25 25>
      constraintSetup {angularLimitsMax=<0 0 0>, angularLimitsMin=<0 0 0>, angularTarget=<1 0 0>}
      localJointOffset=<2 0 0>
     }]
   }
   n95 = Anim Dynamics - Bone: dyn_ribbon_fr_R_a_02!(ComponentPose=@n94, bAlphaBoolEnabled:bool=True, GravityOverride:Vector="0,0,-980", AngularSpringConstant:float=15000, ExternalForce="0,0,0") -> (Pose) {
    alphaInputType=Bool
    angularDampingOverride=1
    angularSpringConstant=15000
    bAngularSpring=true
    bOverrideAngularDamping=true
    bUseGravityOverride=true
    gravityScale=0.35
    simulationSpace=RootRelative
    alphaBoolBlend {blendOption=Linear}
    boundBone {boneName=dyn_ribbon_fr_R_a_02}
    componentLinearAccScale=<0.75 0.75 0.75>
    componentLinearVelScale=<0.75 0.75 0.75>
    gravityOverride {z=-980}
    physicsBodyDefinitions [{
      boundBone {boneName=dyn_ribbon_fr_R_a_02}
      boxExtents=<25 25 25>
      constraintSetup {angularLimitsMax=<0 20 0>, angularLimitsMin=<0 -20 -60>, angularTarget=<1 0 0>}
      localJointOffset=<2 0 0>
     }]
   }
   n96 = Anim Dynamics - Bone: dyn_ribbon_fr_R_a_03!(ComponentPose=@n95, bAlphaBoolEnabled:bool=True, GravityOverride:Vector="0,0,-980", AngularSpringConstant:float=15000, ExternalForce="0,0,0") -> (Pose) {
    alphaInputType=Bool
    angularDampingOverride=1
    angularSpringConstant=15000
    bAngularSpring=true
    bOverrideAngularDamping=true
    bUseGravityOverride=true
    gravityScale=0.35
    simulationSpace=RootRelative
    alphaBoolBlend {blendOption=Linear}
    boundBone {boneName=dyn_ribbon_fr_R_a_03}
    componentLinearAccScale=<0.75 0.75 0.75>
    componentLinearVelScale=<0.75 0.75 0.75>
    gravityOverride {z=-980}
    physicsBodyDefinitions [{
      boundBone {boneName=dyn_ribbon_fr_R_a_03}
      boxExtents=<25 25 25>
      constraintSetup {angularLimitsMax=<0 20 0>, angularLimitsMin=<0 -20 -60>, angularTarget=<1 0 0>}
      localJointOffset=<2 0 0>
     }]
   }
   n97 = Anim Dynamics - Bone: dyn_ribbon_fr_R_b_01!(ComponentPose=@n96, bAlphaBoolEnabled:bool=True, GravityOverride:Vector="0,0,-980", AngularSpringConstant:float=15000, ExternalForce="0,0,0") -> (Pose) {
    alphaInputType=Bool
    angularDampingOverride=1
    angularSpringConstant=15000
    bAngularSpring=true
    bOverrideAngularDamping=true
    bUseGravityOverride=true
    gravityScale=0.35
    simulationSpace=RootRelative
    alphaBoolBlend {blendOption=Linear}
    boundBone {boneName=dyn_ribbon_fr_R_b_01}
    componentLinearAccScale=<0.75 0.75 0.75>
    componentLinearVelScale=<0.75 0.75 0.75>
    gravityOverride {z=-980}
    physicsBodyDefinitions [{
      boundBone {boneName=dyn_ribbon_fr_R_b_01}
      boxExtents=<25 25 25>
      constraintSetup {angularLimitsMax=<0 20 0>, angularLimitsMin=<0 -20 -60>, angularTarget=<1 0 0>}
      localJointOffset=<3 0 0>
     }]
   }
   n98 = Anim Dynamics - Bone: dyn_ribbon_fr_R_b_02!(ComponentPose=@n97, bAlphaBoolEnabled:bool=True, GravityOverride:Vector="0,0,-980", AngularSpringConstant:float=15000, ExternalForce="0,0,0") -> (Pose) {
    alphaInputType=Bool
    angularDampingOverride=1
    angularSpringConstant=15000
    bAngularSpring=true
    bOverrideAngularDamping=true
    bUseGravityOverride=true
    gravityScale=0.35
    simulationSpace=RootRelative
    alphaBoolBlend {blendOption=Linear}
    boundBone {boneName=dyn_ribbon_fr_R_b_02}
    componentLinearAccScale=<0.75 0.75 0.75>
    componentLinearVelScale=<0.75 0.75 0.75>
    gravityOverride {z=-980}
    physicsBodyDefinitions [{
      boundBone {boneName=dyn_ribbon_fr_R_b_02}
      boxExtents=<25 25 25>
      constraintSetup {angularLimitsMax=<0 20 10>, angularLimitsMin=<0 -20 -60>, angularTarget=<1 0 0>}
      localJointOffset=<3 0 0>
     }]
   }
   n99 = Anim Dynamics - Bone: dyn_ribbon_fr_R_b_03!(ComponentPose=@n98, bAlphaBoolEnabled:bool=True, GravityOverride:Vector="0,0,-980", AngularSpringConstant:float=15000, ExternalForce="0,0,0") -> (Pose) {
    alphaInputType=Bool
    angularDampingOverride=1
    angularSpringConstant=15000
    bAngularSpring=true
    bOverrideAngularDamping=true
    bUseGravityOverride=true
    gravityScale=0.35
    simulationSpace=RootRelative
    alphaBoolBlend {blendOption=Linear}
    boundBone {boneName=dyn_ribbon_fr_R_b_03}
    componentLinearAccScale=<0.75 0.75 0.75>
    componentLinearVelScale=<0.75 0.75 0.75>
    gravityOverride {z=-980}
    physicsBodyDefinitions [{
      boundBone {boneName=dyn_ribbon_fr_R_b_03}
      boxExtents=<25 25 25>
      constraintSetup {angularLimitsMax=<0 20 20>, angularLimitsMin=<0 -20 -60>, angularTarget=<1 0 0>}
      localJointOffset=<3 0 0>
     }]
   }
   n100 = Anim Dynamics - Bone: dyn_ribbon_fr_R_c_01!(ComponentPose=@n99, bAlphaBoolEnabled:bool=True, GravityOverride:Vector="0,0,-980", AngularSpringConstant:float=15000, ExternalForce="0,0,0") -> (Pose) {
    alphaInputType=Bool
    angularDampingOverride=1
    angularSpringConstant=15000
    bAngularSpring=true
    bOverrideAngularDamping=true
    bUseGravityOverride=true
    gravityScale=0.35
    simulationSpace=RootRelative
    alphaBoolBlend {blendOption=Linear}
    boundBone {boneName=dyn_ribbon_fr_R_c_01}
    componentLinearAccScale=<0.75 0.75 0.75>
    componentLinearVelScale=<0.75 0.75 0.75>
    gravityOverride {z=-980}
    physicsBodyDefinitions [{
      boundBone {boneName=dyn_ribbon_fr_R_c_01}
      boxExtents=<25 25 25>
      constraintSetup {angularLimitsMax=<0 20 0>, angularLimitsMin=<0 -20 -60>, angularTarget=<1 0 0>}
      localJointOffset=<4 0 0>
     }]
   }
   n101 = Anim Dynamics - Bone: dyn_ribbon_fr_R_c_02!(
    ComponentPose=@n100
    bAlphaBoolEnabled:bool=True
    GravityOverride:Vector="0,0,-980"
    AngularSpringConstant:float=15000
    ExternalForce="0,0,0"
   ) -> (Pose) {
    alphaInputType=Bool
    angularDampingOverride=1
    angularSpringConstant=15000
    bAngularSpring=true
    bOverrideAngularDamping=true
    bUseGravityOverride=true
    gravityScale=0.35
    simulationSpace=RootRelative
    alphaBoolBlend {blendOption=Linear}
    boundBone {boneName=dyn_ribbon_fr_R_c_02}
    componentLinearAccScale=<0.75 0.75 0.75>
    componentLinearVelScale=<0.75 0.75 0.75>
    gravityOverride {z=-980}
    physicsBodyDefinitions [{
      boundBone {boneName=dyn_ribbon_fr_R_c_02}
      boxExtents=<25 25 25>
      constraintSetup {angularLimitsMax=<0 20 10>, angularLimitsMin=<0 -20 -60>, angularTarget=<1 0 0>}
      localJointOffset=<4 0 0>
     }]
   }
   n102 = Anim Dynamics - Bone: dyn_ribbon_fr_R_c_03!(
    ComponentPose=@n101
    bAlphaBoolEnabled:bool=True
    GravityOverride:Vector="0,0,-980"
    AngularSpringConstant:float=15000
    ExternalForce="0,0,0"
   ) -> (Pose) {
    alphaInputType=Bool
    angularDampingOverride=1
    angularSpringConstant=15000
    bAngularSpring=true
    bOverrideAngularDamping=true
    bUseGravityOverride=true
    gravityScale=0.35
    simulationSpace=RootRelative
    alphaBoolBlend {blendOption=Linear}
    boundBone {boneName=dyn_ribbon_fr_R_c_03}
    componentLinearAccScale=<0.75 0.75 0.75>
    componentLinearVelScale=<0.75 0.75 0.75>
    gravityOverride {z=-980}
    physicsBodyDefinitions [{
      boundBone {boneName=dyn_ribbon_fr_R_c_03}
      boxExtents=<25 25 25>
      constraintSetup {angularLimitsMax=<0 20 20>, angularLimitsMin=<0 -20 -60>, angularTarget=<1 0 0>}
      localJointOffset=<4 0 0>
     }]
   }
   n103 = Anim Dynamics - Bone: dyn_ribbon_bk_R_a_01!(ComponentPose=@n102, Alpha=1, GravityOverride:Vector="0,0,-980", AngularSpringConstant:float=12000, ExternalForce="0,0,0") -> (Pose) {
    angularDampingOverride=1
    angularSpringConstant=12000
    bAngularSpring=true
    bOverrideAngularDamping=true
    bUseGravityOverride=true
    gravityScale=0.35
    simulationSpace=RootRelative
    alphaBoolBlend {blendOption=Linear}
    boundBone {boneName=dyn_ribbon_bk_R_a_01}
    componentLinearAccScale=<0.75 0.75 0.75>
    componentLinearVelScale=<0.75 0.75 0.75>
    gravityOverride {z=-980}
    physicsBodyDefinitions [{
      boundBone {boneName=dyn_ribbon_bk_R_a_01}
      boxExtents=<25 25 25>
      constraintSetup {angularLimitsMax=<0 20 0>, angularLimitsMin=<0 -20 -60>, angularTarget=<1 0 0>}
      localJointOffset=<1 0 0>
     }]
   }
   n104 = Anim Dynamics - Bone: dyn_ribbon_bk_R_a_02!(ComponentPose=@n103, Alpha=1, GravityOverride:Vector="0,0,-980", AngularSpringConstant:float=12000, ExternalForce="0,0,0") -> (Pose) {
    angularDampingOverride=1
    angularSpringConstant=12000
    bAngularSpring=true
    bOverrideAngularDamping=true
    bUseGravityOverride=true
    gravityScale=0.35
    simulationSpace=RootRelative
    alphaBoolBlend {blendOption=Linear}
    boundBone {boneName=dyn_ribbon_bk_R_a_02}
    componentLinearAccScale=<0.75 0.75 0.75>
    componentLinearVelScale=<0.75 0.75 0.75>
    gravityOverride {z=-980}
    physicsBodyDefinitions [{
      boundBone {boneName=dyn_ribbon_bk_R_a_02}
      boxExtents=<25 25 25>
      constraintSetup {angularLimitsMax=<0 20 0>, angularLimitsMin=<0 -20 -60>, angularTarget=<1 0 0>}
      localJointOffset=<5 0 0>
     }]
   }
   n105 = Anim Dynamics - Bone: dyn_ribbon_bk_R_b_01!(ComponentPose=@n104, Alpha=1, GravityOverride:Vector="0,0,-980", AngularSpringConstant:float=12000, ExternalForce="0,0,0") -> (Pose) {
    angularDampingOverride=1
    angularSpringConstant=12000
    bAngularSpring=true
    bOverrideAngularDamping=true
    bUseGravityOverride=true
    gravityScale=0.35
    simulationSpace=RootRelative
    alphaBoolBlend {blendOption=Linear}
    boundBone {boneName=dyn_ribbon_bk_R_b_01}
    componentLinearAccScale=<0.75 0.75 0.75>
    componentLinearVelScale=<0.75 0.75 0.75>
    gravityOverride {z=-980}
    physicsBodyDefinitions [{
      boundBone {boneName=dyn_ribbon_bk_R_b_01}
      boxExtents=<25 25 25>
      constraintSetup {angularLimitsMax=<0 20 0>, angularLimitsMin=<0 -20 -60>, angularTarget=<1 0 0>}
      localJointOffset=<1 0 0>
     }]
   }
   n106 = Anim Dynamics - Bone: dyn_ribbon_bk_R_b_02!(ComponentPose=@n105, Alpha=1, GravityOverride:Vector="0,0,-980", AngularSpringConstant:float=12000, ExternalForce="0,0,0") -> (Pose) {
    angularDampingOverride=1
    angularSpringConstant=12000
    bAngularSpring=true
    bOverrideAngularDamping=true
    bUseGravityOverride=true
    gravityScale=0.35
    simulationSpace=RootRelative
    alphaBoolBlend {blendOption=Linear}
    boundBone {boneName=dyn_ribbon_bk_R_b_02}
    componentLinearAccScale=<0.75 0.75 0.75>
    componentLinearVelScale=<0.75 0.75 0.75>
    gravityOverride {z=-980}
    physicsBodyDefinitions [{
      boundBone {boneName=dyn_ribbon_bk_R_b_02}
      boxExtents=<25 25 25>
      constraintSetup {angularLimitsMax=<0 20 0>, angularLimitsMin=<0 -20 -60>, angularTarget=<1 0 0>}
      localJointOffset=<5 0 0>
     }]
   }
   n107 = Anim Dynamics - Bone: dyn_ribbon_bk_R_b_03!(ComponentPose=@n106, Alpha=1, GravityOverride:Vector="0,0,-980", AngularSpringConstant:float=12000, ExternalForce="0,0,0") -> (Pose) {
    angularDampingOverride=1
    angularSpringConstant=12000
    bAngularSpring=true
    bOverrideAngularDamping=true
    bUseGravityOverride=true
    gravityScale=0.35
    simulationSpace=RootRelative
    alphaBoolBlend {blendOption=Linear}
    boundBone {boneName=dyn_ribbon_bk_R_b_03}
    componentLinearAccScale=<0.75 0.75 0.75>
    componentLinearVelScale=<0.75 0.75 0.75>
    gravityOverride {z=-980}
    physicsBodyDefinitions [{
      boundBone {boneName=dyn_ribbon_bk_R_b_03}
      boxExtents=<25 25 25>
      constraintSetup {angularLimitsMax=<0 20 0>, angularLimitsMin=<0 -20 -60>, angularTarget=<1 0 0>}
      localJointOffset=<5 0 0>
     }]
   }
   n108 = Anim Dynamics - Bone: dyn_ribbon_bk_R_c_01!(ComponentPose=@n107, Alpha=1, GravityOverride:Vector="0,0,-980", AngularSpringConstant:float=12000, ExternalForce="0,0,0") -> (Pose) {
    angularDampingOverride=1
    angularSpringConstant=12000
    bAngularSpring=true
    bOverrideAngularDamping=true
    bUseGravityOverride=true
    gravityScale=0.35
    simulationSpace=RootRelative
    alphaBoolBlend {blendOption=Linear}
    boundBone {boneName=dyn_ribbon_bk_R_c_01}
    componentLinearAccScale=<0.75 0.75 0.75>
    componentLinearVelScale=<0.75 0.75 0.75>
    gravityOverride {z=-980}
    physicsBodyDefinitions [{
      boundBone {boneName=dyn_ribbon_bk_R_c_01}
      boxExtents=<25 25 25>
      constraintSetup {angularLimitsMax=<0 20 0>, angularLimitsMin=<0 -20 -60>, angularTarget=<1 0 0>}
      localJointOffset=<1 0 0>
     }]
   }
   n109 = Anim Dynamics - Bone: dyn_ribbon_bk_R_c_02!(ComponentPose=@n108, Alpha=1, GravityOverride:Vector="0,0,-980", AngularSpringConstant:float=12000, ExternalForce="0,0,0") -> (Pose) {
    angularDampingOverride=1
    angularSpringConstant=12000
    bAngularSpring=true
    bOverrideAngularDamping=true
    bUseGravityOverride=true
    gravityScale=0.35
    simulationSpace=RootRelative
    alphaBoolBlend {blendOption=Linear}
    boundBone {boneName=dyn_ribbon_bk_R_c_02}
    componentLinearAccScale=<0.75 0.75 0.75>
    componentLinearVelScale=<0.75 0.75 0.75>
    gravityOverride {z=-980}
    physicsBodyDefinitions [{
      boundBone {boneName=dyn_ribbon_bk_R_c_02}
      boxExtents=<25 25 25>
      constraintSetup {angularLimitsMax=<0 20 0>, angularLimitsMin=<0 -20 -60>, angularTarget=<1 0 0>}
      localJointOffset=<5 0 0>
     }]
   }
   n110 = Anim Dynamics - Bone: dyn_ribbon_bk_R_d_01!(ComponentPose=@n109, Alpha=1, GravityOverride:Vector="0,0,-980", AngularSpringConstant:float=12000, ExternalForce="0,0,0") -> (Pose) {
    angularDampingOverride=1
    angularSpringConstant=12000
    bAngularSpring=true
    bOverrideAngularDamping=true
    bUseGravityOverride=true
    gravityScale=0.35
    simulationSpace=RootRelative
    alphaBoolBlend {blendOption=Linear}
    boundBone {boneName=dyn_ribbon_bk_R_d_01}
    componentLinearAccScale=<0.75 0.75 0.75>
    componentLinearVelScale=<0.75 0.75 0.75>
    gravityOverride {z=-980}
    physicsBodyDefinitions [{
      boundBone {boneName=dyn_ribbon_bk_R_d_01}
      boxExtents=<25 25 25>
      constraintSetup {angularLimitsMax=<0 20 0>, angularLimitsMin=<0 -20 -60>, angularTarget=<1 0 0>}
      localJointOffset=<1 0 0>
     }]
   }
   n111 = Anim Dynamics - Bone: dyn_ribbon_bk_R_d_01(ComponentPose=@n110, Alpha=1, GravityOverride="0,0,-980", AngularSpringConstant=12000, ExternalForce="0,0,0") {
    angularDampingOverride=1
    angularSpringConstant=12000
    bAngularSpring=true
    bOverrideAngularDamping=true
    bUseGravityOverride=true
    gravityScale=0.35
    simulationSpace=RootRelative
    alphaBoolBlend {blendOption=Linear}
    boundBone {boneName=dyn_ribbon_bk_R_d_01}
    componentLinearAccScale=<0.75 0.75 0.75>
    componentLinearVelScale=<0.75 0.75 0.75>
    gravityOverride {z=-980}
    physicsBodyDefinitions [{
      boundBone {boneName=dyn_ribbon_bk_R_d_01}
      boxExtents=<25 25 25>
      constraintSetup {angularLimitsMax=<0 20 0>, angularLimitsMin=<0 -20 -60>, angularTarget=<1 0 0>}
      localJointOffset=<5 0 0>
     }]
   }
   Save cached pose 'AnimDynPose'(Pose=Component To Local(ComponentPose=@n111)) {cacheName=AnimDynPose}
   n112 = Anim Dynamics - Bone: dyn_ribbon_fr_L_a_01!(
    ComponentPose=Local To Component(LocalPose=Use cached pose 'AnimDynPose'())
    Alpha:float=1
    GravityOverride="0,0,-980"
    AngularSpringConstant=12000
    ExternalForce="0,0,0"
   ) -> (Pose) {
    angularDampingOverride=1
    angularSpringConstant=12000
    bAngularSpring=true
    bOverrideAngularDamping=true
    bUseGravityOverride=true
    gravityScale=0.35
    simulationSpace=RootRelative
    alphaBoolBlend {blendOption=Linear}
    boundBone {boneName=dyn_ribbon_fr_L_a_01}
    componentLinearAccScale=<0.75 0.75 0.75>
    componentLinearVelScale=<0.75 0.75 0.75>
    gravityOverride {z=-980}
    physicsBodyDefinitions [{
      boundBone {boneName=dyn_ribbon_fr_L_a_01}
      boxExtents=<25 25 25>
      constraintSetup {angularLimitsMax=<0 20 20>, angularLimitsMin=<0 -10 -20>, angularTarget=<1 0 0>}
      localJointOffset=<2 0 0>
     }]
    simSpaceSettings {maxAngularAcceleration=15, maxAngularVelocity=3, simSpaceAngularAlpha=1}
   }
   n113 = Anim Dynamics - Bone: dyn_ribbon_fr_L_a_02!(ComponentPose=@n112, Alpha:float=1, GravityOverride="0,0,-980", AngularSpringConstant=12000, ExternalForce="0,0,0") -> (Pose) {
    angularDampingOverride=1
    angularSpringConstant=12000
    bAngularSpring=true
    bOverrideAngularDamping=true
    bUseGravityOverride=true
    gravityScale=0.35
    simulationSpace=RootRelative
    alphaBoolBlend {blendOption=Linear}
    boundBone {boneName=dyn_ribbon_fr_L_a_02}
    componentLinearAccScale=<0.75 0.75 0.75>
    componentLinearVelScale=<0.75 0.75 0.75>
    gravityOverride {z=-980}
    physicsBodyDefinitions [{
      boundBone {boneName=dyn_ribbon_fr_L_a_02}
      boxExtents=<25 25 25>
      constraintSetup {angularLimitsMax=<0 20 20>, angularLimitsMin=<0 -10 -20>, angularTarget=<1 0 0>}
      localJointOffset=<2 0 0>
     }]
    simSpaceSettings {maxAngularAcceleration=15, maxAngularVelocity=3, simSpaceAngularAlpha=1}
   }
   n114 = Anim Dynamics - Bone: dyn_ribbon_fr_L_b_01!(ComponentPose=@n113, Alpha:float=1, GravityOverride="0,0,-980", AngularSpringConstant=12000, ExternalForce="0,0,0") -> (Pose) {
    angularDampingOverride=1
    angularSpringConstant=12000
    bAngularSpring=true
    bOverrideAngularDamping=true
    bUseGravityOverride=true
    gravityScale=0.35
    simulationSpace=RootRelative
    alphaBoolBlend {blendOption=Linear}
    boundBone {boneName=dyn_ribbon_fr_L_b_01}
    componentLinearAccScale=<0.75 0.75 0.75>
    componentLinearVelScale=<0.75 0.75 0.75>
    gravityOverride {z=-980}
    physicsBodyDefinitions [{
      boundBone {boneName=dyn_ribbon_fr_L_b_01}
      boxExtents=<25 25 25>
      constraintSetup {angularLimitsMax=<0 20 0>, angularLimitsMin=<0 -10 -20>, angularTarget=<1 0 0>}
      localJointOffset=<6 0 0>
     }]
    simSpaceSettings {maxAngularAcceleration=15, maxAngularVelocity=3, simSpaceAngularAlpha=1}
   }
   n115 = Anim Dynamics - Bone: dyn_ribbon_fr_L_b_02!(ComponentPose=@n114, Alpha:float=1, GravityOverride="0,0,-980", AngularSpringConstant=12000, ExternalForce="0,0,0") -> (Pose) {
    angularDampingOverride=1
    angularSpringConstant=12000
    bAngularSpring=true
    bOverrideAngularDamping=true
    bUseGravityOverride=true
    gravityScale=0.35
    simulationSpace=RootRelative
    alphaBoolBlend {blendOption=Linear}
    boundBone {boneName=dyn_ribbon_fr_L_b_02}
    componentLinearAccScale=<0.75 0.75 0.75>
    componentLinearVelScale=<0.75 0.75 0.75>
    gravityOverride {z=-980}
    physicsBodyDefinitions [{
      boundBone {boneName=dyn_ribbon_fr_L_b_02}
      boxExtents=<25 25 25>
      constraintSetup {angularLimitsMax=<0 20 10>, angularLimitsMin=<0 -20 -20>, angularTarget=<1 0 0>}
      localJointOffset=<4 0 0>
     }]
    simSpaceSettings {maxAngularAcceleration=15, maxAngularVelocity=3, simSpaceAngularAlpha=1}
   }
   n116 = Anim Dynamics - Bone: dyn_ribbon_fr_L_b_03!(ComponentPose=@n115, Alpha:float=1, GravityOverride="0,0,-980", AngularSpringConstant=12000, ExternalForce="0,0,0") -> (Pose) {
    angularDampingOverride=1
    angularSpringConstant=12000
    bAngularSpring=true
    bOverrideAngularDamping=true
    bUseGravityOverride=true
    gravityScale=0.35
    simulationSpace=RootRelative
    alphaBoolBlend {blendOption=Linear}
    boundBone {boneName=dyn_ribbon_fr_L_b_03}
    componentLinearAccScale=<0.75 0.75 0.75>
    componentLinearVelScale=<0.75 0.75 0.75>
    gravityOverride {z=-980}
    physicsBodyDefinitions [{
      boundBone {boneName=dyn_ribbon_fr_L_b_03}
      boxExtents=<25 25 25>
      constraintSetup {angularLimitsMax=<0 20 20>, angularLimitsMin=<0 -20 -20>, angularTarget=<1 0 0>}
      localJointOffset=<2 0 0>
     }]
    simSpaceSettings {maxAngularAcceleration=15, maxAngularVelocity=3, simSpaceAngularAlpha=1}
   }
   n117 = Anim Dynamics - Bone: dyn_ribbon_fr_L_c_01!(ComponentPose=@n116, Alpha:float=1, GravityOverride="0,0,-980", AngularSpringConstant=12000, ExternalForce="0,0,0") -> (Pose) {
    angularDampingOverride=1
    angularSpringConstant=12000
    bAngularSpring=true
    bOverrideAngularDamping=true
    bUseGravityOverride=true
    gravityScale=0.35
    simulationSpace=RootRelative
    alphaBoolBlend {blendOption=Linear}
    boundBone {boneName=dyn_ribbon_fr_L_c_01}
    componentLinearAccScale=<0.75 0.75 0.75>
    componentLinearVelScale=<0.75 0.75 0.75>
    gravityOverride {z=-980}
    physicsBodyDefinitions [{
      boundBone {boneName=dyn_ribbon_fr_L_c_01}
      boxExtents=<25 25 25>
      constraintSetup {angularLimitsMax=<0 0 0>, angularLimitsMin=<0 0 0>, angularTarget=<1 0 0>}
      localJointOffset=<2 0 0>
     }]
    simSpaceSettings {maxAngularAcceleration=15, maxAngularVelocity=3, simSpaceAngularAlpha=1}
   }
   n118 = Anim Dynamics - Bone: dyn_ribbon_fr_L_c_02!(ComponentPose=@n117, Alpha:float=1, GravityOverride="0,0,-980", AngularSpringConstant=12000, ExternalForce="0,0,0") -> (Pose) {
    angularDampingOverride=1
    angularSpringConstant=12000
    bAngularSpring=true
    bOverrideAngularDamping=true
    bUseGravityOverride=true
    gravityScale=0.35
    simulationSpace=RootRelative
    alphaBoolBlend {blendOption=Linear}
    boundBone {boneName=dyn_ribbon_fr_L_c_02}
    componentLinearAccScale=<0.75 0.75 0.75>
    componentLinearVelScale=<0.75 0.75 0.75>
    gravityOverride {z=-980}
    physicsBodyDefinitions [{
      boundBone {boneName=dyn_ribbon_fr_L_c_02}
      boxExtents=<25 25 25>
      constraintSetup {angularLimitsMax=<0 20 20>, angularLimitsMin=<0 -10 -40>, angularTarget=<1 0 0>}
      localJointOffset=<4 0 0>
     }]
    simSpaceSettings {maxAngularAcceleration=15, maxAngularVelocity=3, simSpaceAngularAlpha=1}
   }
   n119 = Anim Dynamics - Bone: dyn_ribbon_fr_L_c_03!(ComponentPose=@n118, Alpha:float=1, GravityOverride="0,0,-980", AngularSpringConstant=12000, ExternalForce="0,0,0") -> (Pose) {
    angularDampingOverride=1
    angularSpringConstant=12000
    bAngularSpring=true
    bOverrideAngularDamping=true
    bUseGravityOverride=true
    gravityScale=0.35
    simulationSpace=RootRelative
    alphaBoolBlend {blendOption=Linear}
    boundBone {boneName=dyn_ribbon_fr_L_c_03}
    componentLinearAccScale=<0.75 0.75 0.75>
    componentLinearVelScale=<0.75 0.75 0.75>
    gravityOverride {z=-980}
    physicsBodyDefinitions [{
      boundBone {boneName=dyn_ribbon_fr_L_c_03}
      boxExtents=<25 25 25>
      constraintSetup {angularLimitsMax=<0 20 20>, angularLimitsMin=<0 -10 -20>, angularTarget=<1 0 0>}
      localJointOffset=<2 0 0>
     }]
    simSpaceSettings {maxAngularAcceleration=15, maxAngularVelocity=3, simSpaceAngularAlpha=1}
   }
   n120 = Anim Dynamics - Bone: dyn_ribbon_fr_R_a_01!(ComponentPose=@n119, Alpha:float=1, GravityOverride="0,0,-980", AngularSpringConstant=15000, ExternalForce="0,0,0") -> (Pose) {
    angularDampingOverride=1
    angularSpringConstant=15000
    bAngularSpring=true
    bOverrideAngularDamping=true
    bUseGravityOverride=true
    gravityScale=0.35
    simulationSpace=RootRelative
    alphaBoolBlend {blendOption=Linear}
    boundBone {boneName=dyn_ribbon_fr_R_a_01}
    componentLinearAccScale=<0.75 0.75 0.75>
    componentLinearVelScale=<0.75 0.75 0.75>
    gravityOverride {z=-980}
    physicsBodyDefinitions [{
      boundBone {boneName=dyn_ribbon_fr_R_a_01}
      boxExtents=<25 25 25>
      constraintSetup {angularLimitsMax=<0 0 0>, angularLimitsMin=<0 0 0>, angularTarget=<1 0 0>}
      localJointOffset=<2 0 0>
     }]
    simSpaceSettings {maxAngularAcceleration=42, maxAngularVelocity=7, simSpaceAngularAlpha=1, externalAngularVelocity {x=2}}
   }
   n121 = Anim Dynamics - Bone: dyn_ribbon_fr_R_a_02!(ComponentPose=@n120, Alpha:float=1, GravityOverride="0,0,-980", AngularSpringConstant=15000, ExternalForce="0,0,0") -> (Pose) {
    angularDampingOverride=1
    angularSpringConstant=15000
    bAngularSpring=true
    bOverrideAngularDamping=true
    bUseGravityOverride=true
    gravityScale=0.35
    simulationSpace=RootRelative
    alphaBoolBlend {blendOption=Linear}
    boundBone {boneName=dyn_ribbon_fr_R_a_02}
    componentLinearAccScale=<0.75 0.75 0.75>
    componentLinearVelScale=<0.75 0.75 0.75>
    gravityOverride {z=-980}
    physicsBodyDefinitions [{
      boundBone {boneName=dyn_ribbon_fr_R_a_02}
      boxExtents=<25 25 25>
      constraintSetup {angularLimitsMax=<0 40 0>, angularLimitsMin=<0 -40 -60>, angularTarget=<1 0 0>}
      localJointOffset=<2 0 0>
     }]
    simSpaceSettings {maxAngularAcceleration=42, maxAngularVelocity=7, simSpaceAngularAlpha=1, externalAngularVelocity {x=2}}
   }
   n122 = Anim Dynamics - Bone: dyn_ribbon_fr_R_a_03!(ComponentPose=@n121, Alpha:float=1, GravityOverride="0,0,-980", AngularSpringConstant=15000, ExternalForce="0,0,0") -> (Pose) {
    angularDampingOverride=1
    angularSpringConstant=15000
    bAngularSpring=true
    bOverrideAngularDamping=true
    bUseGravityOverride=true
    gravityScale=0.35
    simulationSpace=RootRelative
    alphaBoolBlend {blendOption=Linear}
    boundBone {boneName=dyn_ribbon_fr_R_a_03}
    componentLinearAccScale=<0.75 0.75 0.75>
    componentLinearVelScale=<0.75 0.75 0.75>
    gravityOverride {z=-980}
    physicsBodyDefinitions [{
      boundBone {boneName=dyn_ribbon_fr_R_a_03}
      boxExtents=<25 25 25>
      constraintSetup {angularLimitsMax=<0 40 0>, angularLimitsMin=<0 -40 -60>, angularTarget=<1 0 0>}
      localJointOffset=<2 0 0>
     }]
    simSpaceSettings {maxAngularAcceleration=42, maxAngularVelocity=7, simSpaceAngularAlpha=1, externalAngularVelocity {x=2}}
   }
   n123 = Anim Dynamics - Bone: dyn_ribbon_fr_R_b_01!(ComponentPose=@n122, Alpha:float=1, GravityOverride="0,0,-980", AngularSpringConstant=15000, ExternalForce="0,0,0") -> (Pose) {
    angularDampingOverride=1
    angularSpringConstant=15000
    bAngularSpring=true
    bOverrideAngularDamping=true
    bUseGravityOverride=true
    gravityScale=0.35
    simulationSpace=RootRelative
    alphaBoolBlend {blendOption=Linear}
    boundBone {boneName=dyn_ribbon_fr_R_b_01}
    componentLinearAccScale=<0.75 0.75 0.75>
    componentLinearVelScale=<0.75 0.75 0.75>
    gravityOverride {z=-980}
    physicsBodyDefinitions [{
      boundBone {boneName=dyn_ribbon_fr_R_b_01}
      boxExtents=<25 25 25>
      constraintSetup {angularLimitsMax=<0 40 0>, angularLimitsMin=<0 -40 -60>, angularTarget=<1 0 0>}
      localJointOffset=<2 0 0>
     }]
    simSpaceSettings {maxAngularAcceleration=42, maxAngularVelocity=7, simSpaceAngularAlpha=1, externalAngularVelocity {x=2}}
   }
   n124 = Anim Dynamics - Bone: dyn_ribbon_fr_R_b_02!(ComponentPose=@n123, Alpha:float=1, GravityOverride="0,0,-980", AngularSpringConstant=15000, ExternalForce="0,0,0") -> (Pose) {
    angularDampingOverride=1
    angularSpringConstant=15000
    bAngularSpring=true
    bOverrideAngularDamping=true
    bUseGravityOverride=true
    gravityScale=0.35
    simulationSpace=RootRelative
    alphaBoolBlend {blendOption=Linear}
    boundBone {boneName=dyn_ribbon_fr_R_b_02}
    componentLinearAccScale=<0.75 0.75 0.75>
    componentLinearVelScale=<0.75 0.75 0.75>
    gravityOverride {z=-980}
    physicsBodyDefinitions [{
      boundBone {boneName=dyn_ribbon_fr_R_b_02}
      boxExtents=<25 25 25>
      constraintSetup {angularLimitsMax=<0 40 10>, angularLimitsMin=<0 -40 -60>, angularTarget=<1 0 0>}
      localJointOffset=<2 0 0>
     }]
    simSpaceSettings {maxAngularAcceleration=42, maxAngularVelocity=7, simSpaceAngularAlpha=1, externalAngularVelocity {x=2}}
   }
   n125 = Anim Dynamics - Bone: dyn_ribbon_fr_R_b_03!(ComponentPose=@n124, Alpha:float=1, GravityOverride="0,0,-980", AngularSpringConstant=15000, ExternalForce="0,0,0") -> (Pose) {
    angularDampingOverride=1
    angularSpringConstant=15000
    bAngularSpring=true
    bOverrideAngularDamping=true
    bUseGravityOverride=true
    gravityScale=0.35
    simulationSpace=RootRelative
    alphaBoolBlend {blendOption=Linear}
    boundBone {boneName=dyn_ribbon_fr_R_b_03}
    componentLinearAccScale=<0.75 0.75 0.75>
    componentLinearVelScale=<0.75 0.75 0.75>
    gravityOverride {z=-980}
    physicsBodyDefinitions [{
      boundBone {boneName=dyn_ribbon_fr_R_b_03}
      boxExtents=<25 25 25>
      constraintSetup {angularLimitsMax=<0 40 20>, angularLimitsMin=<0 -40 -60>, angularTarget=<1 0 0>}
      localJointOffset=<2 0 0>
     }]
    simSpaceSettings {maxAngularAcceleration=42, maxAngularVelocity=7, simSpaceAngularAlpha=1, externalAngularVelocity {x=2}}
   }
   n126 = Anim Dynamics - Bone: dyn_ribbon_fr_R_c_01!(ComponentPose=@n125, Alpha:float=1, GravityOverride="0,0,-980", AngularSpringConstant=15000, ExternalForce="0,0,0") -> (Pose) {
    angularDampingOverride=1
    angularSpringConstant=15000
    bAngularSpring=true
    bOverrideAngularDamping=true
    bUseGravityOverride=true
    gravityScale=0.35
    simulationSpace=RootRelative
    alphaBoolBlend {blendOption=Linear}
    boundBone {boneName=dyn_ribbon_fr_R_c_01}
    componentLinearAccScale=<0.75 0.75 0.75>
    componentLinearVelScale=<0.75 0.75 0.75>
    gravityOverride {z=-980}
    physicsBodyDefinitions [{
      boundBone {boneName=dyn_ribbon_fr_R_c_01}
      boxExtents=<25 25 25>
      constraintSetup {angularLimitsMax=<0 40 0>, angularLimitsMin=<0 -40 -60>, angularTarget=<1 0 0>}
      localJointOffset=<3 0 0>
     }]
    simSpaceSettings {maxAngularAcceleration=42, maxAngularVelocity=7, simSpaceAngularAlpha=1, externalAngularVelocity {x=2}}
   }
   n127 = Anim Dynamics - Bone: dyn_ribbon_fr_R_c_02!(ComponentPose=@n126, Alpha:float=1, GravityOverride="0,0,-980", AngularSpringConstant=15000, ExternalForce="0,0,0") -> (Pose) {
    angularDampingOverride=1
    angularSpringConstant=15000
    bAngularSpring=true
    bOverrideAngularDamping=true
    bUseGravityOverride=true
    gravityScale=0.35
    simulationSpace=RootRelative
    alphaBoolBlend {blendOption=Linear}
    boundBone {boneName=dyn_ribbon_fr_R_c_02}
    componentLinearAccScale=<0.75 0.75 0.75>
    componentLinearVelScale=<0.75 0.75 0.75>
    gravityOverride {z=-980}
    physicsBodyDefinitions [{
      boundBone {boneName=dyn_ribbon_fr_R_c_02}
      boxExtents=<25 25 25>
      constraintSetup {angularLimitsMax=<0 40 10>, angularLimitsMin=<0 -40 -60>, angularTarget=<1 0 0>}
      localJointOffset=<3 0 0>
     }]
    simSpaceSettings {maxAngularAcceleration=42, maxAngularVelocity=7, simSpaceAngularAlpha=1, externalAngularVelocity {x=2}}
   }
   n128 = Anim Dynamics - Bone: dyn_ribbon_fr_R_c_03!(ComponentPose=@n127, Alpha:float=1, GravityOverride="0,0,-980", AngularSpringConstant=15000, ExternalForce="0,0,0") -> (Pose) {
    angularDampingOverride=1
    angularSpringConstant=15000
    bAngularSpring=true
    bOverrideAngularDamping=true
    bUseGravityOverride=true
    gravityScale=0.35
    simulationSpace=RootRelative
    alphaBoolBlend {blendOption=Linear}
    boundBone {boneName=dyn_ribbon_fr_R_c_03}
    componentLinearAccScale=<0.75 0.75 0.75>
    componentLinearVelScale=<0.75 0.75 0.75>
    gravityOverride {z=-980}
    physicsBodyDefinitions [{
      boundBone {boneName=dyn_ribbon_fr_R_c_03}
      boxExtents=<25 25 25>
      constraintSetup {angularLimitsMax=<0 40 20>, angularLimitsMin=<0 -40 -60>, angularTarget=<1 0 0>}
      localJointOffset=<2 0 0>
     }]
    simSpaceSettings {maxAngularAcceleration=42, maxAngularVelocity=7, simSpaceAngularAlpha=1, externalAngularVelocity {x=2}}
   }
   Save cached pose 'FrontEndPose'(Pose=Component To Local(ComponentPose=@n128)) {cacheName=FrontEndPose}
   Trail controller - Bone: dyn_ribbon_L_02(Alpha=1) {
    bLimitRotation=true
    bUsePlanarLimit=true
    debugLifeTime=0.1
    maxDeltaTime=0.0333
    alphaScaleBiasClamp {bInterpResult=true}
    baseJoint {boneName=root}
    planarLimits [{drivingBone {boneName=pack}, planeTransform {location=<0 0 0>, rotation=<rot 0 0 -90>, scale=<1 1 1>}}]
    rotationLimits [{limitMax=<0 10 0>, limitMin=<0 -10 -20>}, {limitMax=<0 10 0>, limitMin=<0 -10 -40>}]
    rotationOffsets [<0 0 0>, <0 0 0>]
    trailBone {boneName=dyn_ribbon_L_02}
    trailRelaxationSpeed {editorCurveData {keys [<at 0 23 RCIM_Linear>,<at 0.284 19.3799 RCIM_Linear>,<at 0.6554 8.5924 RCIM_Linear>,<at 1 5 RCIM_Linear>]}}
   }
   Get BwdMovement() -> (BwdMovement:double)
   Get SkydivingAlpha() -> (SkydivingAlpha:Rotator)
   Get Is Not in Frontend() -> (Is Not in Frontend:bool)
   Get JumpingMovement()
   Get bIsCrouching()
   Get bIsSkydiveFloating()
  ]}
 wingPosition {
  entry=Base
  path=AnimGraph.AnimGraphNode_StateMachine_0.wingPosition
  ids=id
  nodes [
   AnimStateNode_0 = Base() {is=title AnimStateNode}
   AnimStateNode_1 = Jumping() {is=title AnimStateNode}
   AnimStateNode_2 = Falling() {is=title AnimStateNode}
   AnimStateNode_3 = Targeting() {is=title AnimStateNode}
   AnimStateNode_5 = SkydiveMode() {is=title AnimStateNode}
   AnimStateNode_6 = Crouching() {is=title AnimStateNode}
   AnimStateNode_9 = Tactical Sprint() {is=title AnimStateNode}
   AnimStateTransitionNode_0 = AnimStateTransitionNode() {
    blendMode=Custom
    crossfadeDuration=0.5
    customBlendCurve=/Game/Accessories/FORT_Backpacks/M_MED_Demon_Pack/Animation/JumpTransitionBlendCurve
    priorityOrder=2
   }
   AnimStateTransitionNode_1 = AnimStateTransitionNode() {
    blendMode=Custom
    crossfadeDuration=0.5
    customBlendCurve=/Game/Accessories/FORT_Backpacks/M_MED_Demon_Pack/Animation/JumpTransitionBlendOutCurve
   }
   AnimStateTransitionNode_10 = AnimStateTransitionNode()
   AnimStateTransitionNode_11 = AnimStateTransitionNode()
   AnimStateTransitionNode_14 = AnimStateTransitionNode()
   AnimStateTransitionNode_15 = AnimStateTransitionNode()
   AnimStateTransitionNode_16 = AnimStateTransitionNode()
   AnimStateTransitionNode_18 = AnimStateTransitionNode()
   AnimStateTransitionNode_2 = AnimStateTransitionNode() {crossfadeDuration=0.3}
   AnimStateTransitionNode_3 = AnimStateTransitionNode() {
    blendMode=Custom
    crossfadeDuration=0.3
    customBlendCurve=&JumpTransitionBlendOutCurve
   }
   AnimStateTransitionNode_4 = AnimStateTransitionNode() {
    blendMode=Custom
    crossfadeDuration=0.3
    customBlendCurve=&JumpTransitionBlendCurve
   }
   AnimStateTransitionNode_5 = AnimStateTransitionNode()
   AnimStateTransitionNode_6 = AnimStateTransitionNode() {crossfadeDuration=0.25}
   AnimStateTransitionNode_7 = AnimStateTransitionNode() {crossfadeDuration=0.25}
   AnimStateTransitionNode_8 = AnimStateTransitionNode()
   AnimStateTransitionNode_9 = AnimStateTransitionNode()
  ]
  links [
   {
    dst=Crouching
    rule=AnimGraph.AnimGraphNode_StateMachine_0.wingPosition.AnimStateTransitionNode_16.AnimationTransitionGraph_0
    src=Base
    via=AnimStateTransitionNode_16
   }
   {
    dst=Falling
    rule=AnimGraph.AnimGraphNode_StateMachine_0.wingPosition.AnimStateTransitionNode_4.Transition
    src=Base
    via=AnimStateTransitionNode_4
   }
   {
    dst=Jumping
    rule=AnimGraph.AnimGraphNode_StateMachine_0.wingPosition.AnimStateTransitionNode_0.Transition
    src=Base
    via=AnimStateTransitionNode_0
   }
   {
    dst=SkydiveMode
    rule=AnimGraph.AnimGraphNode_StateMachine_0.wingPosition.AnimStateTransitionNode_15.Transition
    src=Base
    via=AnimStateTransitionNode_15
   }
   {
    dst=Tactical Sprint
    rule=AnimGraph.AnimGraphNode_StateMachine_0.wingPosition.AnimStateTransitionNode_8.Transition
    src=Base
    via=AnimStateTransitionNode_8
   }
   {
    dst=Targeting
    rule=AnimGraph.AnimGraphNode_StateMachine_0.wingPosition.AnimStateTransitionNode_6.Transition
    src=Base
    via=AnimStateTransitionNode_6
   }
   {
    dst=Base
    rule=AnimGraph.AnimGraphNode_StateMachine_0.wingPosition.AnimStateTransitionNode_18.AnimationTransitionGraph_1
    src=Crouching
    via=AnimStateTransitionNode_18
   }
   {
    dst=Base
    rule=AnimGraph.AnimGraphNode_StateMachine_0.wingPosition.AnimStateTransitionNode_3.Transition
    src=Falling
    via=AnimStateTransitionNode_3
   }
   {
    dst=Base
    rule=AnimGraph.AnimGraphNode_StateMachine_0.wingPosition.AnimStateTransitionNode_1.Transition
    src=Jumping
    via=AnimStateTransitionNode_1
   }
   {
    dst=Falling
    rule=AnimGraph.AnimGraphNode_StateMachine_0.wingPosition.AnimStateTransitionNode_2.Transition
    src=Jumping
    via=AnimStateTransitionNode_2
   }
   {
    dst=SkydiveMode
    rule=AnimGraph.AnimGraphNode_StateMachine_0.wingPosition.AnimStateTransitionNode_5.Transition
    src=Jumping
    via=AnimStateTransitionNode_5
   }
   {
    dst=Tactical Sprint
    rule=AnimGraph.AnimGraphNode_StateMachine_0.wingPosition.AnimStateTransitionNode_14.Transition
    src=Jumping
    via=AnimStateTransitionNode_14
   }
   {
    dst=Base
    rule=AnimGraph.AnimGraphNode_StateMachine_0.wingPosition.AnimStateTransitionNode_11.Transition
    src=SkydiveMode
    via=AnimStateTransitionNode_11
   }
   {
    dst=Base
    rule=AnimGraph.AnimGraphNode_StateMachine_0.wingPosition.AnimStateTransitionNode_9.Transition
    src=Tactical Sprint
    via=AnimStateTransitionNode_9
   }
   {
    dst=Jumping
    rule=AnimGraph.AnimGraphNode_StateMachine_0.wingPosition.AnimStateTransitionNode_10.Transition
    src=Tactical Sprint
    via=AnimStateTransitionNode_10
   }
   {
    dst=Base
    rule=AnimGraph.AnimGraphNode_StateMachine_0.wingPosition.AnimStateTransitionNode_7.Transition
    src=Targeting
    via=AnimStateTransitionNode_7
   }
  ]
 }
 Base {
  path=AnimGraph.AnimGraphNode_StateMachine_0.wingPosition.AnimStateNode_0.Base
  nodes [
   n1 = Get PlayerVector_FwdBwd() -> (PlayerVector_FwdBwd:double)
   n2 = Get PlayerVector_LR() -> (PlayerVector_LR:double)
   n3 = Wing_Movement_Poses_PoseAsset : Base!(PoseWeight:float=1) -> (Pose:Pose Link Structure) {
    is=title AnimGraphNode_PoseByName
    poseAsset=/BRCosmetics/Accessories/FORT_Backpacks/Backpack_F_MED_TangyRadishFlame/Animation/Wing_Movement_Poses_PoseAsset
    poseName=Base
   }
   Output Animation Pose(Result=Apply Additive(
     Base=@n3
     Additive=Blendspace Player 'BS_Wing_Leans_Ground'(X=@n2, Y=@n1) {blendSpace=/BRCosmetics/Accessories/FORT_Backpacks/Backpack_F_MED_TangyRadishFlame/Animation/BS_Wing_Leans_Ground}
    ) {alphaBoolBlend {blendOption=Linear}})
  ]
 }
 Jumping {
  path=AnimGraph.AnimGraphNode_StateMachine_0.wingPosition.AnimStateNode_1.Jumping
  nodes [
   n1 = Get PlayerVector_FwdBwd()
   n2 = Get PlayerVector_LR()
   n3 = Wing_Movement_Poses_PoseAsset : Jump!(PoseWeight:float=1) -> (Pose:Pose Link Structure) {
    is=title AnimGraphNode_PoseByName
    poseAsset=&Wing_Movement_Poses_PoseAsset
    poseName=Jump
   }
   Output Animation Pose(Result=Apply Additive(
     Base=@n3
     Additive=Blendspace Player 'BS_Wing_Leans_Ground'(X=@n2, Y=@n1) {blendSpace=/BRCosmetics/Accessories/FORT_Backpacks/M_MED_Demon_Pack/Animation/BS_Wing_Leans_Ground}
    ) {alphaBoolBlend {blendOption=Linear}})
  ]
 }
 Falling {
  path=AnimGraph.AnimGraphNode_StateMachine_0.wingPosition.AnimStateNode_2.Falling
  nodes [
   n1 = Get PlayerVector_FwdBwd()
   n2 = Get PlayerVector_LR()
   n3 = Wing_Movement_Poses_PoseAsset : Jump(PoseWeight=1) {poseAsset=&Wing_Movement_Poses_PoseAsset, poseName=Jump}
   Output Animation Pose(Result=Apply Additive(
     Base=@n3
     Additive=Blendspace Player 'BS_Wing_Leans_Ground'(X=@n2, Y=@n1) {blendSpace=/BRCosmetics/Accessories/FORT_Backpacks/Backpack_F_MED_TangyRadishFlame/Animation/BS_Wing_Leans_Ground}
    ) {alphaBoolBlend {blendOption=Linear}})
  ]
 }
 Targeting {
  path=AnimGraph.AnimGraphNode_StateMachine_0.wingPosition.AnimStateNode_3.Targeting
  nodes [
   n1 = Wing_Movement_Poses_PoseAsset : Targeting!(PoseWeight:float=1) -> (Pose:Pose Link Structure) {
    is=title AnimGraphNode_PoseByName
    poseAsset=&Wing_Movement_Poses_PoseAsset
    poseName=Targeting
   }
   Output Animation Pose(Result=@n1)
  ]
 }
 SkydiveMode {
  path=AnimGraph.AnimGraphNode_StateMachine_0.wingPosition.AnimStateNode_5.SkydiveMode
  nodes [
   AnimGraphNode_StateMachine_0 = Skydiving State() {is=title AnimGraphNode_StateMachine}
   Output Animation Pose(Result=@AnimGraphNode_StateMachine_0)
  ]
 }
 Skydiving State {
  entry=Parachuting
  path=AnimGraph.AnimGraphNode_StateMachine_0.wingPosition.AnimStateNode_5.SkydiveMode.AnimGraphNode_StateMachine_0.Skydiving State
  ids=id
  nodes [
   AnimStateNode_1 = Diving() {is=title AnimStateNode}
   AnimStateNode_2 = Floating() {is=title AnimStateNode}
   AnimStateNode_4 = Parachuting() {is=title AnimStateNode}
   AnimStateTransitionNode_1 = AnimStateTransitionNode() {crossfadeDuration=0.5}
   AnimStateTransitionNode_10 = AnimStateTransitionNode()
   AnimStateTransitionNode_12 = AnimStateTransitionNode()
   AnimStateTransitionNode_13 = AnimStateTransitionNode()
   AnimStateTransitionNode_3 = AnimStateTransitionNode()
   AnimStateTransitionNode_5 = AnimStateTransitionNode() {crossfadeDuration=0.5}
  ]
  links [
   {
    dst=Floating
    rule=AnimGraph.AnimGraphNode_StateMachine_0.wingPosition.AnimStateNode_5.SkydiveMode.AnimGraphNode_StateMachine_0.Skydiving State.AnimStateTransitionNode_1.AnimationTransitionGraph_0
    src=Diving
    via=AnimStateTransitionNode_1
   }
   {
    dst=Parachuting
    rule=AnimGraph.AnimGraphNode_StateMachine_0.wingPosition.AnimStateNode_5.SkydiveMode.AnimGraphNode_StateMachine_0.Skydiving State.AnimStateTransitionNode_10.AnimationTransitionGraph_1
    src=Diving
    via=AnimStateTransitionNode_10
   }
   {
    dst=Diving
    rule=AnimGraph.AnimGraphNode_StateMachine_0.wingPosition.AnimStateNode_5.SkydiveMode.AnimGraphNode_StateMachine_0.Skydiving State.AnimStateTransitionNode_5.AnimationTransitionGraph_0
    src=Floating
    via=AnimStateTransitionNode_5
   }
   {
    dst=Parachuting
    rule=AnimGraph.AnimGraphNode_StateMachine_0.wingPosition.AnimStateNode_5.SkydiveMode.AnimGraphNode_StateMachine_0.Skydiving State.AnimStateTransitionNode_3.AnimationTransitionGraph_1
    src=Floating
    via=AnimStateTransitionNode_3
   }
   {
    dst=Diving
    rule=AnimGraph.AnimGraphNode_StateMachine_0.wingPosition.AnimStateNode_5.SkydiveMode.AnimGraphNode_StateMachine_0.Skydiving State.AnimStateTransitionNode_13.AnimationTransitionGraph_1
    src=Parachuting
    via=AnimStateTransitionNode_13
   }
   {
    dst=Floating
    rule=AnimGraph.AnimGraphNode_StateMachine_0.wingPosition.AnimStateNode_5.SkydiveMode.AnimGraphNode_StateMachine_0.Skydiving State.AnimStateTransitionNode_12.AnimationTransitionGraph_1
    src=Parachuting
    via=AnimStateTransitionNode_12
   }
  ]
 }
 Diving {
  path=AnimGraph.AnimGraphNode_StateMachine_0.wingPosition.AnimStateNode_5.SkydiveMode.AnimGraphNode_StateMachine_0.Skydiving State.AnimStateNode_1.Diving
  nodes [
   n1 = Get PlayerVector_FwdBwd()
   n2 = Get PlayerVector_LR()
   n3 = Wing_Movement_Poses_PoseAsset : Dive!(PoseWeight:float=1) -> (Pose:Pose Link Structure) {
    is=title AnimGraphNode_PoseByName
    poseAsset=&Wing_Movement_Poses_PoseAsset
    poseName=Dive
   }
   Output Animation Pose(Result=Apply Additive(
     Base=@n3
     Additive=Blendspace Player 'BS_Wing_Leans_Air'(X=@n2, Y=@n1) {blendSpace=/BRCosmetics/Accessories/FORT_Backpacks/Backpack_F_MED_TangyRadishFlame/Animation/BS_Wing_Leans_Air}
    ) {alphaBoolBlend {blendOption=Linear}})
  ]
 }
 Floating {
  path=AnimGraph.AnimGraphNode_StateMachine_0.wingPosition.AnimStateNode_5.SkydiveMode.AnimGraphNode_StateMachine_0.Skydiving State.AnimStateNode_2.Floating
  nodes [
   n1 = Get PlayerVector_FwdBwd()
   n2 = Get PlayerVector_LR()
   n3 = Wing_Movement_Poses_PoseAsset : Float!(PoseWeight:float=1) -> (Pose:Pose Link Structure) {
    is=title AnimGraphNode_PoseByName
    poseAsset=&Wing_Movement_Poses_PoseAsset
    poseName=Float
   }
   Output Animation Pose(Result=Apply Additive(
     Base=@n3
     Additive=Blendspace Player 'BS_Wing_Leans_Air'(X=@n2, Y=@n1) {blendSpace=&BS_Wing_Leans_Air}
    ) {alphaBoolBlend {blendOption=Linear}})
  ]
 }
 Parachuting {
  path=AnimGraph.AnimGraphNode_StateMachine_0.wingPosition.AnimStateNode_5.SkydiveMode.AnimGraphNode_StateMachine_0.Skydiving State.AnimStateNode_4.Parachuting
  nodes [
   n1 = Get PlayerVector_FwdBwd()
   n2 = Get PlayerVector_LR()
   n3 = Wing_Movement_Poses_PoseAsset : Float(PoseWeight=1) {poseAsset=&Wing_Movement_Poses_PoseAsset, poseName=Float}
   Output Animation Pose(Result=Apply Additive(
     Base=@n3
     Additive=Blendspace Player 'BS_Wing_Leans_Air'(X=@n2, Y=@n1) {blendSpace=&BS_Wing_Leans_Air}
    ) {alphaBoolBlend {blendOption=Linear}})
  ]
 }
 AnimationTransitionGraph_0 {
  path=AnimGraph.AnimGraphNode_StateMachine_0.wingPosition.AnimStateNode_5.SkydiveMode.AnimGraphNode_StateMachine_0.Skydiving State.AnimStateTransitionNode_1.AnimationTransitionGraph_0
  nodes [Result(bCanEnterTransition=NOT Boolean(A=Get bIsSkydiveDiveMode()))]
 }
 AnimationTransitionGraph_1 {
  path=AnimGraph.AnimGraphNode_StateMachine_0.wingPosition.AnimStateNode_5.SkydiveMode.AnimGraphNode_StateMachine_0.Skydiving State.AnimStateTransitionNode_10.AnimationTransitionGraph_1
  nodes [Result(bCanEnterTransition=Get bIsParachuteOpened())]
 }
 AnimationTransitionGraph_1 {
  path=AnimGraph.AnimGraphNode_StateMachine_0.wingPosition.AnimStateNode_5.SkydiveMode.AnimGraphNode_StateMachine_0.Skydiving State.AnimStateTransitionNode_12.AnimationTransitionGraph_1
  nodes [Result(bCanEnterTransition=NOT Boolean(A=Get bIsParachuteOpened()))]
 }
 AnimationTransitionGraph_1 {
  path=AnimGraph.AnimGraphNode_StateMachine_0.wingPosition.AnimStateNode_5.SkydiveMode.AnimGraphNode_StateMachine_0.Skydiving State.AnimStateTransitionNode_13.AnimationTransitionGraph_1
  nodes [Result(bCanEnterTransition=NOT Boolean(A=Get bIsParachuteOpened()))]
 }
 AnimationTransitionGraph_1 {
  path=AnimGraph.AnimGraphNode_StateMachine_0.wingPosition.AnimStateNode_5.SkydiveMode.AnimGraphNode_StateMachine_0.Skydiving State.AnimStateTransitionNode_3.AnimationTransitionGraph_1
  nodes [Result(bCanEnterTransition=Get bIsParachuteOpened())]
 }
 AnimationTransitionGraph_0 {
  path=AnimGraph.AnimGraphNode_StateMachine_0.wingPosition.AnimStateNode_5.SkydiveMode.AnimGraphNode_StateMachine_0.Skydiving State.AnimStateTransitionNode_5.AnimationTransitionGraph_0
  nodes [Result(bCanEnterTransition=Get bIsSkydiveDiveMode())]
 }
 Crouching {
  path=AnimGraph.AnimGraphNode_StateMachine_0.wingPosition.AnimStateNode_6.Crouching
  nodes [
   n1 = Wing_Movement_Poses_PoseAsset : Crouch!(PoseWeight:float=1) -> (Pose:Pose Link Structure) {
    is=title AnimGraphNode_PoseByName
    poseAsset=&Wing_Movement_Poses_PoseAsset
    poseName=Crouch
   }
   Output Animation Pose(Result=@n1)
  ]
 }
 Tactical Sprint {
  path=AnimGraph.AnimGraphNode_StateMachine_0.wingPosition.AnimStateNode_9.Tactical Sprint
  nodes [
   n1 = Wing_Movement_Poses_PoseAsset : Sprint!(PoseWeight:float=1) -> (Pose:Pose Link Structure) {
    is=title AnimGraphNode_PoseByName
    poseAsset=&Wing_Movement_Poses_PoseAsset
    poseName=Sprint
   }
   Output Animation Pose(Result=@n1)
  ]
 }
 Transition {
  path=AnimGraph.AnimGraphNode_StateMachine_0.wingPosition.AnimStateTransitionNode_0.Transition
  nodes [Result(bCanEnterTransition=Get bIsJumping())]
 }
 Transition {
  path=AnimGraph.AnimGraphNode_StateMachine_0.wingPosition.AnimStateTransitionNode_1.Transition
  nodes [
   Result(bCanEnterTransition=NOT Boolean(A=Get bIsJumping()))
   Get bShouldPredictLanding() -> (bShouldPredictLanding:bool)
   Get bLandingPredicted() -> (bLandingPredicted:bool)
   Get bDebugLandPrediction() -> (bDebugLandPrediction:bool)
   Get bIsFalling()
  ]
 }
 Transition {
  path=AnimGraph.AnimGraphNode_StateMachine_0.wingPosition.AnimStateTransitionNode_10.Transition
  nodes [Result(bCanEnterTransition=Get bIsJumping())]
 }
 Transition {
  path=AnimGraph.AnimGraphNode_StateMachine_0.wingPosition.AnimStateTransitionNode_11.Transition
  nodes [Result(bCanEnterTransition=NOT Boolean(A=Get bIsSkyDiving()))]
 }
 Transition {
  path=AnimGraph.AnimGraphNode_StateMachine_0.wingPosition.AnimStateTransitionNode_14.Transition
  nodes [Result(bCanEnterTransition=NOT Boolean(A=Get bIsJumping()))]
 }
 Transition {
  path=AnimGraph.AnimGraphNode_StateMachine_0.wingPosition.AnimStateTransitionNode_15.Transition
  nodes [Result(bCanEnterTransition=Get bIsSkyDiving())]
 }
 AnimationTransitionGraph_0 {
  path=AnimGraph.AnimGraphNode_StateMachine_0.wingPosition.AnimStateTransitionNode_16.AnimationTransitionGraph_0
  nodes [
   Result(bCanEnterTransition=Get Is Crouching or Emoting())
   OR Boolean(A=Get bIsCrouching(), B=Get bIsPlayingEmote())
  ]
 }
 AnimationTransitionGraph_1 {
  path=AnimGraph.AnimGraphNode_StateMachine_0.wingPosition.AnimStateTransitionNode_18.AnimationTransitionGraph_1
  nodes [
   Result(bCanEnterTransition=NOT Boolean(A=Get Is Crouching or Emoting()))
   OR Boolean(A=Get bIsCrouching(), B=Get bIsPlayingEmote())
  ]
 }
 Transition {
  path=AnimGraph.AnimGraphNode_StateMachine_0.wingPosition.AnimStateTransitionNode_2.Transition
  nodes [Result(bCanEnterTransition=Get bIsFalling())]
 }
 Transition {
  path=AnimGraph.AnimGraphNode_StateMachine_0.wingPosition.AnimStateTransitionNode_3.Transition
  nodes [Result(bCanEnterTransition=NOT Boolean(A=Get bIsFalling()))]
 }
 Transition {
  path=AnimGraph.AnimGraphNode_StateMachine_0.wingPosition.AnimStateTransitionNode_4.Transition
  nodes [Result(bCanEnterTransition=Get bIsFalling())]
 }
 Transition {
  path=AnimGraph.AnimGraphNode_StateMachine_0.wingPosition.AnimStateTransitionNode_5.Transition
  nodes [Result(bCanEnterTransition=Get bIsSkyDiving())]
 }
 Transition {
  path=AnimGraph.AnimGraphNode_StateMachine_0.wingPosition.AnimStateTransitionNode_6.Transition
  nodes [Result(bCanEnterTransition=Get bIsTargeting())]
 }
 Transition {
  path=AnimGraph.AnimGraphNode_StateMachine_0.wingPosition.AnimStateTransitionNode_7.Transition
  nodes [Result(bCanEnterTransition=NOT Boolean(A=Get bIsTargeting()))]
 }
 Transition {
  path=AnimGraph.AnimGraphNode_StateMachine_0.wingPosition.AnimStateTransitionNode_8.Transition
  nodes [Result(bCanEnterTransition=Get bIsTacticalSprinting())]
 }
 Transition {
  path=AnimGraph.AnimGraphNode_StateMachine_0.wingPosition.AnimStateTransitionNode_9.Transition
  nodes [Result(bCanEnterTransition=NOT Boolean(A=Get bIsTacticalSprinting()))]
 }
 EventGraph {nodes [
   Event BlueprintUpdateAnimation()
   > n1 = Sequence!(execute) -> (then_0, then_1, then_2:Exec, then_3:Exec, then_4:Exec, then_5:Exec)
   > n2 = "Gravity Override State Function (Nativized)"!(
    execute
    CharacterPartAnimInstance:Custom Character Part Anim Instance Object Reference=Self-Reference()
    GravityOverrideParameterSet:Fort Gravity Override Parameter Set Structure=Get Gravity Override Params_0() -> (Gravity Override Params_0:Fort Gravity Override Parameter Set Structure)
    RigidBodyAnimNodeParameterSet:Fort Rigid Body Anim Node Parameter Set Structure
    DebugDraw:bool=false
    GameTimeInSeconds:float=GetGameTimeInSeconds()
    DebugJointName:Name=root
    StateLogicLayerName:Name=none
   ) -> (
    then
    OutGravityOverride:Vector
    OutCurrentState:Name=None
    OutGravityOverrideParameters:Fort Gravity Override Parameters Structure
    OutRigidBodyAnimNodeParameters:Fort Rigid Body Anim Node Parameters Structure
   )
   > Set GravityOverride(GravityOverride=@n2)
   > n3 = Branch(Condition=Get bIsInFrontEnd())
   > Set GravityOverride(GravityOverride="0,0,-980")
   Set GravityOverride(execute=@n3.else, GravityOverride="0,0,-980")
   > n4 = Branch(Condition=Get bIsSkydiveFloating())
   > Set GravityOverride(GravityOverride="0,0,-500")
   Set GravityOverride(execute=@n4.else, GravityOverride="0,0,-980")
   n5 = PlayerMovement(execute=@n1.then_1, AnimInstance=Self-Reference(), XDotProductCoefficient=1, YDotProductCoefficient=1, ZDotProductCoefficient=1)
   > Set PlayerVector_LR(PlayerVector_LR:double=@n5) -> (Output_Get:double)
   > Set PlayerVector_FwdBwd(PlayerVector_FwdBwd:double=@n5.YMovement) -> (Output_Get:double)
   n6 = NOT Boolean(A=Get bIsSkyDiving())
   n7 = NOT Boolean(A=Get bIsOnGround() -> (bIsOnGround:bool))
   > Set JumpingForwardAlpha(JumpingForwardAlpha:double=float * float(A:bool=AND Boolean(A=@n7, B=@n6), B=Get PlayerVector_FwdBwd())) -> (Output_Get:double)
   n8 = NOT Boolean(A=Get bIsSkyDiving())
   n9 = NOT Boolean(A=Get bIsFalling())
   n10 = NOT Boolean(A=Get bIsOnGround())
   > Set JumpingMovement(JumpingMovement:bool=AND Boolean(A=Get bIsMoving2D(), B=AND Boolean!(A=@n9, B=@n8, C:bool=@n10) -> (ReturnValue))) -> (Output_Get:bool)
   n11 = Get bIsSliding()
   > Set CrouchStrafe(CrouchStrafe:bool=AND Boolean(A=Get bIsCrouching(), B=Get bIsMoving2D(), C=NOT Boolean(A=@n11))) -> (Output_Get:bool)
   n12 = Get bIsCrouching()
   > Set IsMovingStanding(IsMovingStanding:bool=AND Boolean(A=Get bIsMoving2D(), B=Get bIsOnGround(), C=NOT Boolean(A=@n12))) -> (Output_Get:bool)
   n13 = Get bIsSkyDiving()
   n14 = Get PlayerVector_LR()
   n15 = Get bIsFalling()
   n16 = Get bIsJumping()
   > Set Jumping RL(Jumping RL:double=float * float!(A:double="Absolute (Float)"(A=@n14), B:bool=OR Boolean(A=@n15, B=@n16), C:bool=NOT Boolean(A=@n13)) -> (ReturnValue)) -> (Output_Get:double)
   n17 = Get bIsSkyDiving()
   n18 = Get bIsJumping()
   n19 = Get bIsFalling()
   > Set JumpingFalling(JumpingFalling:bool=AND Boolean!(A=OR Boolean(A=@n18, B=@n19), B=NOT Boolean(A=@n17)) -> (ReturnValue)) -> (Output_Get:bool)
   n20 = Get PlayerVector_LR()
   > Set RightMovement(RightMovement:double=float * float!(A=@n20, B=Get IsMovingStanding()) -> (ReturnValue)) -> (Output_Get:double)
   > Set LeftMovement(LeftMovement:double=float * float(A=@n20, B:double=-1)) -> (Output_Get:double)
   > Set ForwardGroundMovement(ForwardGroundMovement:double=float * float(A=Get PlayerVector_FwdBwd(), B:bool=Get bIsOnGround())) -> (Output_Get:double)
   > Set Is Crouching or Emoting(Is Crouching or Emoting:bool=OR Boolean(A=Get bIsCrouching(), B=Get bIsPlayingEmote())) -> (Output_Get:bool)
   n21 = Branch(execute=@n1.then_2, Condition=Get bIsInFrontEnd())
   > Set BwdMovement(BwdMovement:double=0.5) -> (Output_Get:double)
   n22 = PlayerMovement(execute=@n21.else, AnimInstance=Self-Reference(), XDotProductCoefficient=1, YDotProductCoefficient=1, ZDotProductCoefficient=1)
   > n23 = Branch(Condition=Get bIsDBNOCarried())
   > Set BwdMovement(BwdMovement=0.5)
   n24 = Branch(execute=@n23.else, Condition=AND Boolean(A=Get bIsTargeting(), B=Get bIsCrouching()))
   n25 = MapRangeClamped(
    Value=float + float(A="Absolute (Float)"(A=@n22), B=MapRangeClamped(Value=@n22.YMovement, InRangeA=-1, InRangeB=1, OutRangeA=0.3)) {is=title K2Node_CommutativeAssociativeBinaryOperator}
    InRangeA=0
    InRangeB=1.65
   )
   > Set BwdMovement(BwdMovement=float * float(A=@n25, B:double=0.25))
   Set BwdMovement(execute=@n24.else, BwdMovement=float * float(A=@n25, B=0.8))
   n26 = Branch(execute=@n1.then_3, Condition=Get bIsSkyDiving())
   > n27 = Branch(Condition="float > float"(A=Get PawnVelocityZ(), B=50) {is=title K2Node_CallFunction})
   > Set BwdMovement(BwdMovement=float * float(A=Get BwdMovement(), B=0) {is=title K2Node_CommutativeAssociativeBinaryOperator})
   Set BwdMovement(execute=[@n26.else, @n27.else], BwdMovement=float * float(A=Get BwdMovement(), B=1))
   n28 = NOT Boolean(A=Get bIsCrouching())
   n29 = NOT Boolean(A=Get bIsMoving2D())
   n30 = Get bIsOnGround()
   Set StandingIdle(
    execute=@n1.then_4
    StandingIdle:bool=OR Boolean(A=AND Boolean!(A=@n30, B=@n28, C:bool=@n29, D:bool) -> (ReturnValue), B=Get bIsInFrontEnd())
   ) -> (Output_Get:bool)
   > n31 = Branch(Condition=Get bIsInFrontEnd())
   > Set Angular Spring Constant R(Angular Spring Constant R:double=12000) -> (Output_Get:double)
   > Set Angular Spring Constant(Angular Spring Constant:double=9000) -> (Output_Get:double)
   Set Angular Spring Constant R(execute=@n31.else, Angular Spring Constant R=15000)
   > Set Angular Spring Constant(Angular Spring Constant=12000)
   > Set Is Not in Frontend(Is Not in Frontend:bool=true) -> (Output_Get:bool)
   > Branch(Condition=Get bIsPlayingEmote())
   > Set Angular Spring Constant R(Angular Spring Constant R=50000)
   > Set Angular Spring Constant(Angular Spring Constant=30000)
   n32 = GetAnimInstance(self=GetOwnerSkeletalMeshComponent())
   Is Valid(exec=@n1.then_5, InputObject=@n32)
   > Set IsTangyRadishEmote(IsTangyRadishEmote:bool=OR Boolean!(
     A=Montage_IsPlaying(self=@n32, Montage=/BRCosmetics/Animation/Game/MainPlayer/Emotes/TangyRadishMagma/Emote_TangyRadishMagma_CLF_M)
     B=Montage_IsPlaying(self=@n32, Montage=/BRCosmetics/Animation/Game/MainPlayer/Emotes/TangyRadishMagma/Emote_TangyRadishMagma_CLM_M)
     C:bool=Montage_IsPlaying(self=@n32, Montage=/BRCosmetics/Animation/Game/MainPlayer/Emotes/TangyRadishMagma/Emote_TangyRadishMagma_CMF_M)
     D:bool=Montage_IsPlaying(self=@n32, Montage=/BRCosmetics/Animation/Game/MainPlayer/Emotes/TangyRadishMagma/Emote_TangyRadishMagma_CMM_M)
    ) -> (ReturnValue)) -> (Output_Get:bool)
   Event OnBeginSkydiving()
   > Set DynamicsAlpha(DynamicsAlpha:double=1) -> (Output_Get:double)
   > Set SkydivingAlpha(SkydivingAlpha:Rotator="0,-20,0") -> (Output_Get:Rotator)
   Event OnEndSkydiving()
   > Set DynamicsAlpha(DynamicsAlpha=0.5)
   > Set SkydivingAlpha(SkydivingAlpha="0, 0, 0")
   TryGetPawnOwner()
  ]}
]
variables [
 Angular Spring Constant {type=number}
 Angular Spring Constant R {type=number}
 BwdMovement {type=number}
 CrouchStrafe {type=boolean}
 DynamicsAlpha {type=number}
 ForwardGroundMovement {type=number}
 Gravity Override Params {type=GravityOverrideParamsStruct}
 Gravity Override Params_0 {
  type=FortGravityOverrideParameterSet
  default=(Skydive_GoingUp=(GravityOverride=()),Skydive_FloatingDown=(GravityOverride=()),Skydive_DivingDown=(GravityOverride=()),Skydive_Parachuting=(GravityOverride=()),OnGround_Standing=(GravityOverride=()),OnGround_Moving=(GravityOverride=()),OnGround_Moving_HighRotationRate=(GravityOverride=()),OnGround_Crouch_Moving=(GravityOverride=()),Emote_or_Melee=(GravityOverride=()),Swimming=(GravityOverride=()),Front_End=(GravityOverride=(GravityOverride=(X=0.000000,Y=0.000000,Z=-980.000000))),DBNO=(GravityOverride=()),DBNO_Carried=(GravityOverride=()),OnGround_Falling=(GravityOverride=()),Ziplining=(GravityOverride=()),Skydive_Surfing=(GravityOverride=()),Riding_Standing=(GravityOverride=()),Riding_Moving=(GravityOverride=()),Riding_Moving_Targeting=(GravityOverride=()),Tactical_Sprint=(GravityOverride=()),Rail_Grinding=(GravityOverride=()),)
 }
 GravityOverride {type=Vector}
 Is Crouching or Emoting {type=boolean}
 Is Not in Frontend {type=boolean}
 IsMovingStanding {type=boolean}
 IsTangyRadishEmote {type=boolean}
 Jumping RL {type=number}
 JumpingFalling {type=boolean}
 JumpingForwardAlpha {type=number}
 JumpingMovement {type=boolean}
 LeftMovement {type=number}
 PlayerVector_FwdBwd {type=number}
 PlayerVector_LR {type=number}
 RightMovement {type=number}
 SkydiveFloat movement {type=boolean}
 SkydivingAlpha {type=Rotator}
 StandingIdle {type=boolean}
]
